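# Translucent, unlit, two-sided spline material in the Meteorwomen dive contrail effects folder, apparently derived from a Reality Sapling effect.
# Emissive layers two panning line and caustic textures, distorted by scrolling noise, over tinted background colors, while UV gradient and sphere masks shape opacity.
# World position offset combines screen-aware spline thickening, noise along normals and hit-effect deformation.
# A SaplingTeleport parameter sweeps a local-axis mask with world-aligned noise, driving dissolve, glow and outward offset.
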Material /BRCosmetics/Athena/Effects/Particles/Skydiving/NS_Contrail_Dive_Meteorwomen/M_RealitySapling_Spline_3
graph {
 properties {
  bUsedWithInstancedStaticMeshes=true
  bUsedWithNiagaraMeshParticles=true
  bUsedWithSplineMeshes=true
  blendMode=BLEND_Translucent
  pixelDepthOffsetMode=PDOM_Legacy
  shadingModel=MSM_Unlit
  twoSided=true
  usedShadingModels=Unlit
 }
 nodes [
  n1 = Constant3Vector() {constant=<rgba 0 0.6106 1 0>}
  n2 = PreSkinnedPosition()
  n3 = ShaderStageSwitch(PixelShader=VertexInterpolator(VS=@n2).PS, VertexShader=@n2)
  n4 = ChannelMaskParameter(None=Divide(A=@n3, B=PreSkinnedLocalBounds().Max)) {group=Teleport, maskChannel=Blue, parameterName="Z-CHanne;l", defaultValue=<rgba 0 0 1 0>}
  n5 = ScalarParameter() {defaultValue=1, group=Teleport, parameterName=SaplingTeleport}
  n6 = Subtract(A=Add(A=OneMinus(None=@n4.R)), B=Multiply(A=OneMinus(None=@n5)) {constB=2})
  n7 = Time()
  n8 = VectorParameter() {group=Teleport, parameterName=PanSpeedAndDirection, defaultValue=<rgba 0 0 64 1>}
  n9 = MaterialFunctionCall(
   TextureObject=TextureObject() {texture=/Game/Athena/SafeZone/Textures/T_JanusShield3_M}
   TextureSize=VectorParameter() {group=Teleport, parameterName=CellNoiseSize, defaultValue=<rgba 90 90 90 1>}.RGB
   WorldPosition=Add(A=NamedRerouteUsage(), B=Multiply(A=@n7, B=@n8.RGB))
  ) {materialFunction=/Game/Weapons/WeaponSkins/Wraps/MaterialFunctions/WorldAlignedTexture}
  n10 = MaterialFunctionCall(
   Transition Phase=SphereMask(A=@n6, B=Constant() {r=0.5}, Radius=ScalarParameter() {defaultValue=0.5, parameterName=1, sliderMin=1}, Hardness=Constant())
   Height Texture=ChannelMaskParameter(None=@n9.XYZ Texture) {group=Teleport, parameterName=NoiseChannel}.R
   Contrast=Constant()
  ) {materialFunction=&HeightLerp}
  n11 = Power(Base=Multiply(A=Multiply(A=@n1, B=@n10.Alpha), B=ScalarParameter() {defaultValue=4, group=Teleport, parameterName=GlowStrength}))
  Add(B=@n11)
  Constant() {r=1}
  n12 = ScalarParameter() {group=Emissive, parameterName=Intensity}
  n13 = NamedRerouteUsage()
  n14 = NamedRerouteUsage()
  n15 = TextureSampleParameter2D(UVs=Panner(Coordinate=Add(A=@n13, B=@n14), Time=NamedRerouteUsage(), Speed=NamedRerouteUsage()) {speedX=1, speedY=1}) {group=Emissive, parameterName=LineTexture_01, texture=/Game/Athena/Environments/Nexus/Textures/T_ZeroPoint_Line_Stars}
  n16 = NamedRerouteUsage()
  n17 = NamedRerouteUsage()
  n18 = TextureSampleParameter2D(UVs=Panner(Coordinate=Subtract(A=@n16, B=@n17), Time=NamedRerouteUsage(), Speed=NamedRerouteUsage()) {speedX=1, speedY=1}) {group=Emissive, parameterName=LineTexture_02, texture=/Game/Effects/Fort_Effects/Textures/Tile/T_Caustics_Chromatic}
  n19 = Multiply(A=@n18.RGB, B=VectorParameter() {group=Emissive, parameterName=SecondaryColorTint, defaultValue=<rgba 1.6172 0.1146 2 1>}.RGB) {constB=0.5}
  n20 = TextureSampleParameter2D(UVs=Panner(Coordinate=NamedRerouteUsage(), Time=NamedRerouteUsage(), Speed=NamedRerouteUsage()) {speedX=1, speedY=1}) {group=Distortion, parameterName=NoiseTexture, texture=&T_Caustics_Chromatic}
  n21 = MaterialFunctionCall(
   In=Add(A=@n20.RGB, B=ScalarParameter() {group=Emissive, parameterName=Emissive_Bias})
   Contrast=ScalarParameter() {defaultValue=0.09, group=Emissive, parameterName=Emissive_Contrast}
  ) {materialFunction=&CheapContrast}
  n22 = VectorParameter() {group=Emissive, parameterName=ZipGlow, defaultValue=<rgba 0.625 10.0013 20 0>}
  n23 = ScalarParameter() {defaultValue=10, group=Emissive, parameterName=ZipGLow_Intensity}
  n24 = Add(A=LinearInterpolate(A=Multiply(A=@n22.RGB, B=@n23), Alpha=NamedRerouteUsage()) {constA=5, constB=0}) {constB=0.5}
  n25 = VectorParameter() {group=Emissive, parameterName=BackgroundColor1, defaultValue=<rgba 1.6825 0.1859 3 0>}
  n26 = VectorParameter() {group=Emissive, parameterName=BackgroundColor2, defaultValue=<rgba 0.0081 0.0162 0.1 0>}
  n27 = NamedRerouteUsage()
  n28 = Add(
   A=Add(A=Add(A=@n15.RGB, B=@n19), B=Multiply(A=@n21.Result, B=@n24) {constB=0.5})
   B=Multiply(A=LinearInterpolate(A=@n25.RGB, B=@n26.RGB, Alpha=@n27), B=NamedRerouteUsage())
  )
  n29 = ScalarParameter() {defaultValue=1, group=Emissive, parameterName=EmissiveMult}
  n30 = Add(A=Multiply(A=Add(A=@n12), B=Multiply(A=@n28, B=@n29)), B=NamedRerouteUsage())
  MaterialFunctionCall(Emissive=@n30) {materialFunction=&MF_FX_EmissiveAdjustments}
  n31 = VertexNormalWS()
  n32 = Multiply(A=Time(), B=ScalarParameter() {group=WPO_Nurturing, parameterName=Debug Time})
  Multiply(
   A=@n31
   B=MaterialFunctionCall(Input=Sine(None=@n32), Input Low=Constant() {r=-1}, Input High=Constant() {r=1}, Target Low=Constant(), Target High=Constant() {r=0.5}) {materialFunction=&RemapValueRange}.Result
  )
  n33 = Fresnel()
  Multiply(A=@n10.Alpha, B=Clamp(None=Power(Base=@n33)) {minDefault=0.5})
  n34 = VectorParameter() {group=Emissive, parameterName=Emissive2_UVTile, defaultValue=<rgba 0.05 2.5 0 0>}
  NamedRerouteDeclaration(None=Multiply(A=TextureCoordinate() {coordinateIndex=1}, B=AppendVector(A=@n34.R, B=@n34.R))) {name=Emissive2UVTile, nodeColor=<rgba 0 1 0.2031 1>}
  NamedRerouteDeclaration(None=Multiply(A=Time(), B=ScalarParameter() {defaultValue=0.15, group=Emissive, parameterName=Emissive2_PanSpeed})) {name=Emissive2UVTime, nodeColor=<rgba 0 1 0.8828 1>}
  NamedRerouteDeclaration(None=Multiply(A=Time(), B=ScalarParameter() {defaultValue=0.15, group=Emissive, parameterName=Emissive1_PanSpeed})) {name=Emissive1UVTime, nodeColor=<rgba 0.5938 0 1 1>}
  NamedRerouteDeclaration(None=Multiply(A=ComponentMask(None=@n20.RGB) {g=true, r=true}) {constB=0.05}) {name=Distortion, nodeColor=<rgba 1 0.9844 0 1>}
  n35 = VectorParameter() {group=Emissive, parameterName=Emissive1_UVTile, defaultValue=<rgba 0.01 1 0 0>}
  NamedRerouteDeclaration(None=Multiply(A=TextureCoordinate() {coordinateIndex=1}, B=AppendVector(A=@n35.R, B=@n35.R))) {name=Emissive1UVTile, nodeColor=<rgba 1 0 0.6094 1>}
  n36 = VertexColor()
  NamedRerouteDeclaration(None=ConstantBiasScale(None=Sine(None=@n36.R) {period=0.7})) {name=UVVertColorOffsets, nodeColor=<rgba 0.4062 1 0 1>}
  n37 = MaterialFunctionCall(UV Channel=TextureCoordinate()) {materialFunction=&LinearGradient}
  n38 = OneMinus(None=@n37.UGradient)
  n39 = TextureCoordinate()
  n40 = Power(Base=OneMinus(None=ComponentMask(None=@n39) {g=true}), Exp=ScalarParameter() {defaultValue=1, group=Opacity, parameterName=PowerBottomFade})
  n41 = Add(A=NamedRerouteUsage(), B=Add(A=NamedRerouteUsage(), B=NamedRerouteUsage()))
  n42 = Constant() {r=0.5}
  n43 = SphereMask(A=ComponentMask(None=TextureCoordinate()) {r=true}, B=@n42, Radius=@n42) {hardnessPercent=0}
  n44 = Constant() {r=0.25}
  n45 = Multiply(A=NamedRerouteDeclaration(None=MaterialFunctionCall(In=@n43, Contrast=@n44) {materialFunction=&CheapContrast}.Result) {name=OpacitySphereMask, nodeColor=<rgba 1 0.4922 0 1>}) {constB=0.4}
  n46 = Multiply(A=Multiply(A=Add(A=@n41, B=@n45) {constB=0.5}, B=@n43))
  n47 = StaticSwitchParameter(True=Multiply(A=Multiply(A=@n37.UGradient, B=@n38), B=Multiply(A=@n40, B=@n46)), False=@n46) {defaultValue=true, parameterName=UseTopBottomFade?}
  n48 = ScalarParameter() {defaultValue=1, group=Opacity, parameterName=OpacityMult}
  NamedRerouteDeclaration(None=Multiply(A=Multiply(A=@n47, B=@n48), B=NamedRerouteUsage())) {name=Opacity, nodeColor=<rgba 0.0078 0 1 1>}
  n49 = ViewSize()
  n50 = Divide(
   A=PixelDepth().R
   B=Multiply(A=ComponentMask(None=@n49) {g=true}, B=ScalarParameter() {defaultValue=0.2, group=WPO, parameterName=SplineScreenWidth})
  )
  n51 = Add(A=ScalarParameter() {defaultValue=4, group=WPO, parameterName=SplineWidthBase}, B=@n50)
  n52 = Add(A=ScalarParameter() {defaultValue=0.1, group=WPO, parameterName=SplineWidthTip}, B=@n50)
  n53 = ScalarParameter() {defaultValue=1, group=WPO_Nurturing, parameterName=WPO Amount Opening}
  n54 = TextureCoordinate() {coordinateIndex=1}
  n55 = Multiply(A=ComponentMask(None=@n54) {r=true}, B=ScalarParameter() {group=WPO_Noise, parameterName=WPO Tile U})
  n56 = Multiply(A=ComponentMask(None=@n54) {g=true}, B=ScalarParameter() {defaultValue=0.5, group=WPO_Noise, parameterName=WPO Tile V})
  n57 = ScalarParameter() {group=WPO_Noise, parameterName=WPO Pan Speed U}
  n58 = ScalarParameter() {defaultValue=0.1, group=WPO_Noise, parameterName=WPO Pan Speed V}
  n59 = TextureSample(UVs=Panner(Coordinate=AppendVector(A=@n55, B=@n56), Time=Time(), Speed=AppendVector(A=@n57, B=@n58)) {speedX=-1, speedY=0.1}) {samplerType=SAMPLERTYPE_Grayscale, texture=&LowResBlurredNoise}
  n60 = ScalarParameter() {group=WPO_Noise, parameterName=NoiseContrast}
  n61 = Multiply(
   A=Multiply(A=@n31, B=Add(A=@n53))
   B=Multiply(
    A=ScalarParameter() {defaultValue=20, group=WPO_Noise, parameterName=WPO Amnt Noise}
    B=MaterialFunctionCall(In=@n59.R, Contrast=@n60) {materialFunction=&CheapContrast}.Result
   )
  )
  n62 = ComponentMask(None=TextureCoordinate()) {g=true}
  n63 = MaterialFunctionCall(
   In=Add(A=OneMinus(None=@n62), B=ScalarParameter() {defaultValue=0.25, group=WPO, parameterName=WPO Bottom Mask Offset})
   Contrast=ScalarParameter() {defaultValue=0.5, group=WPO, parameterName=WPO Bottom Mask Contrast}
  ) {materialFunction=&CheapContrast}
  n64 = Add(
   A=Add(A=Multiply(A=@n61, B=@n63.Result)) {constB=0}
   B=MaterialFunctionCall!(
    WSDeformationCenter=VectorParameter() {bUseCustomPrimitiveData=true, group=z_code, parameterName=WSDeformationCenter, primitiveDataIndex=20}.RGB
    Code_DestructionAmount=ScalarParameter() {group=WPO_Hits, parameterName=Code_DestructionAmount}
    Code_UseUvForDamage
    Code_UseHitEffects=StaticBoolParameter() {defaultValue=true, group=WPO_Hits, parameterName=Code_UseHitEffects}
    Code_Emissive
    WSDamageRadius=ScalarParameter() {bUseCustomPrimitiveData=true, defaultValue=1, group=z_code, parameterName=WSDamageRadius, primitiveDataIndex=26}
    CrackDamageShader
    Diffuse
    ScalarModDamageCracks
    ScalarOffsetforDamageCracks
    ExposeLerpFunctionForInvertingTheHitDistanceCalc
    InvertDamageFalloff
    CrackWorldPosition
    WorldPosition
    DistanceCullFadeAnimationOffset=ScalarParameter() {bUseCustomPrimitiveData=true, group=z_code, parameterName=DistanceCullFadeAnimationOffset, primitiveDataIndex=12}
   ) {materialFunction=/Game/Packages/jl_tests/Fortnite_HitEffects}.Hit And Shrink WPO
  )
  n65 = NamedRerouteUsage()
  NamedRerouteDeclaration(None=Add(
    A=MaterialFunctionCall(WidthBase=@n51, WidthTip=@n52) {materialFunction=&SplineThicken}.WorldPositionOffset
    B=Add(A=@n64, B=@n65)
   )) {name=WPO, nodeColor=<rgba 1 0 0.2812 1>}
  n66 = VectorParameter() {group=Distortion, parameterName=Distortion_UVTile, defaultValue=<rgba 0.015 8 0 0>}
  NamedRerouteDeclaration(None=Multiply(A=TextureCoordinate() {coordinateIndex=1}, B=AppendVector(A=@n66.R, B=@n66.R))) {name=DistortionUVTile, nodeColor=<rgba 0.875 1 0 1>}
  n67 = ScalarParameter() {defaultValue=0.15, group=Distortion, parameterName=Distortion_PanSpeed}
  n68 = VertexColor()
  NamedRerouteDeclaration(None=Multiply(A=Time(), B=Add(A=@n67, B=@n68.G))) {name=DistortionUVTime, nodeColor=<rgba 0 1 0.9766 1>}
  n69 = VectorParameter() {group=Distortion, parameterName=Distortion_PanDirection, defaultValue=<rgba 0.2 1 0 0>}
  NamedRerouteDeclaration(None=AppendVector(A=@n69.R, B=@n69.R)) {name=DistortionUVPanDirection, nodeColor=<rgba 1 0 0.5625 1>}
  NamedRerouteDeclaration(None=MaterialFunctionCall(
    Transition Phase=@n6
    Height Texture=ChannelMaskParameter(None=@n9.XYZ Texture) {group=Teleport, maskChannel=Green, parameterName=OPacityNoiseChannel, defaultValue=<rgba 0 1 0 0>}.R
    Contrast=ScalarParameter() {defaultValue=1, group=Teleport, parameterName=Contrastr}
   ) {materialFunction=&HeightLerp}.Alpha) {name=TeleportOpacityMask, nodeColor=<rgba 1 0 0.1406 1>}
  n70 = VectorParameter() {group=Emissive, parameterName=Emissive2_PanDirection, defaultValue=<rgba 0.25 0.5 0 0>}
  NamedRerouteDeclaration(None=AppendVector(A=@n70.R, B=@n70.R)) {name=Emissive2PanDirection, nodeColor=<rgba 0.3359 0 1 1>}
  NamedRerouteDeclaration(None=@n11) {name=Teleport Emissive, nodeColor=<rgba 0 0.9062 1 1>}
  n71 = Normalize(VectorInput=Subtract(A=WorldPosition().XYZ, B=ObjectPositionWS()))
  n72 = VectorParameter() {group=Teleport, parameterName=Scale_AmtTele, defaultValue=<rgba 32 32 0 1>}
  n73 = Constant() {r=0.25}
  n74 = Constant() {r=0.5}
  n75 = Constant()
  NamedRerouteDeclaration(None=LinearInterpolate(B=Multiply(A=@n71, B=@n72.RGB), Alpha=SphereMask(A=@n6, B=@n73, Radius=@n74, Hardness=@n75))) {name=Teleport WPO, nodeColor=<rgba 1 0.8203 0 1>}
  NamedRerouteDeclaration(None=@n3) {name=PreskinnedPos, nodeColor=<rgba 0.6641 1 0 1>}
  NamedRerouteDeclaration(None=@n15.B) {name=TextureSample_B, nodeColor=<rgba 0 1 0.5781 1>}
  NamedRerouteDeclaration(None=@n18.B) {name=TextureSample2_B, nodeColor=<rgba 0.5703 0 1 1>}
  NamedRerouteDeclaration(None=ComponentMask(None=@n21.Result) {g=true, r=true}) {name=TextureSample3_B, nodeColor=<rgba 1 0.0234 0 1>}
  NamedRerouteDeclaration(None=@n30) {name=Emissive, nodeColor=<rgba 1 0 0.0703 1>}
  n76 = VectorParameter() {group=Emissive, parameterName=Emissive1_PanDirection, defaultValue=<rgba 0 1 0 0>}
  NamedRerouteDeclaration(None=AppendVector(A=@n76.R, B=@n76.R)) {name=Emissive1PanDirection, nodeColor=<rgba 1 0 0.5625 1>}
  NamedRerouteUsage()
  NamedRerouteUsage()
  NamedRerouteUsage()
  OneMinus(None=@n10.Alpha)
 ]
}
