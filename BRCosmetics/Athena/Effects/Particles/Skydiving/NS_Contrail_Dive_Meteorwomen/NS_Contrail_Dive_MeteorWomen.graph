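# Niagara skydiving contrail for the MeteorWomen cosmetic, adding glows, mesh trails, embers and shards to a diving character.
# System-level diving attributes, integrators and a timeline drive offsets, scale, opacity and forces in the emitters.
# Most emitters pin particles to skeletal mesh bones, including a bowl mesh, coronas, trails and a MiniZeroPoint orb with light glow.
# Embers simulate on GPU with curl noise, while fragment bursts size themselves from a SaplingAge user parameter.
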NiagaraSystem /BRCosmetics/Athena/Effects/Particles/Skydiving/NS_Contrail_Dive_Meteorwomen/NS_Contrail_Dive_MeteorWomen
dependencies {
 usedDataInterfaces [
  &NiagaraDataInterfaceCurve
  &NiagaraDataInterfaceColorCurve
  &NiagaraDataInterfaceArrayInt32
  &NiagaraDataInterfaceArrayFloat
  &NiagaraDataInterfaceArrayFloat3
 ]
 usedDynamicInputs [
  &Divide_Float
  &ReturnParticleIDAttributes
  &NormalizeVector
  &FN_FrontEndFloatModifier
  &TransformVector
  &SetBoolByFloatComparison
  &Subtract_Float
 ]
 usedModules [
  &SystemState
  /Game/Effects/Niagara/Modules/Skydiving/FN_DivingAttributes/v5/FN_DivingAttributes
  &self:SystemSpawnScript.SystemScriptSource.SystemScriptGraph.NiagaraNodeAssignment_0.SetVariables_#1
  &self:SystemSpawnScript.SystemScriptSource.SystemScriptGraph.NiagaraNodeAssignment_1.SetVariables_#2
  &FN_Integrator
  &self:SystemSpawnScript.SystemScriptSource.SystemScriptGraph.NiagaraNodeAssignment_2.SetVariables_#3
  &self:SystemSpawnScript.SystemScriptSource.SystemScriptGraph.NiagaraNodeAssignment_5.SetVariables_#4
  &Timeline
  &self:SystemSpawnScript.SystemScriptSource.SystemScriptGraph.NiagaraNodeAssignment_3.SetVariables_#5
 ]
}
emitters {
 Bowl_Multi {
  emitterSpawnScript [&FN_PreserveSkydiveUserVariables() {module=PreserveSkydiveUserVariables}]
  emitterUpdateScript [
   &EmitterState()
   &SpawnBurst_Instantaneous(Spawn Count=1, Use Spawn Probability=false) {enabled=false}
   &SpawnRate(SpawnRate=2, Use Spawn Probability=false)
  ]
  particleSpawnScript [
   &FN_SetSpawnGroup(Number of Spawn Groups=1) {enabled=false, module=FN_SetSpawnGroup001}
   /Niagara/Modules/Spawn/Initialization/InitializeParticle(
    Color=<rgba 3 0.6228 2.0597 1>
    Mesh Scale=<1.5 1.7 1.7>
    Module.Write Color=true
    Module.Write Lifetime=true
    Module.Write Mass=false
    Module.Write Material Random=true
    Module.Write Position=false
    Module.Write Scale=true
    Module.Write SpriteRotation=false
    Module.Write SpriteSize=false
    Position=&SimulationPosition()
    Sprite Rotation=&UniformRangedFloat(Minimum=0, Maximum=1)
   )
   &InitialMeshOrientation(
    Orientation Axis=<0 0 1>
    Rotation=&MakeVector(X=&RandomRangeFloat(Minimum=-1, Maximum=1), Y=0, Z=0)
    Use Orientation Vector=true
    Use Rotation Vector=true
   )
  ]
  particleUpdateScript [
   /Game/Effects/Niagara/Modules/Skydiving/FN_DivingAttributes/FN_DivingAttributes_v2(Accumulate Dive=0.1, Accumulate Rest=0.1, SkeletalMesh:NiagaraDataInterfaceSkeletalMesh=$User.SkeletalMesh) {module=FN_DivingAttributes_v2001}
   &SkeletalMeshLocation(
    Bone / Socket Index=4
    Bone Sampling Mode=::NewEnumerator4
    Inherit Velocity=false
    Interpolate Mesh Sampling=false
    MaskSpawnGroup=false
    Sampled Position Offset=&Lerp_Vector(A=<0 0 -40>, B=<0 0 -60>, Alpha=$System.Dive)
    Skeletal Mesh=$User.SkeletalMesh
   )
   &ParticleState()
   &SolveForcesAndVelocity(Clamp Velocity=false, Limit Acceleration=false)
   &ScaleColor(
    Scale Alpha=&FloatFromCurve(
     FloatCurve=DataInterface() {propertyValues {Curve {keys [
         {arriveTangent=4.0516, interpMode=RCIM_Cubic, leaveTangent=4.0516, tangentMode=RCTM_User, time=0, value=0}
         {arriveTangent=0.0961, interpMode=RCIM_Cubic, leaveTangent=0.0961, tangentMode=RCTM_User, time=0.25, value=0.5}
         {arriveTangent=0.2272, interpMode=RCIM_Cubic, leaveTangent=0.2272, tangentMode=RCTM_User, time=1, value=0}
        ]}}}
     Scale Curve=&Lerp_Float(A=0.3, B=0.9, Alpha=$System.FN_Integrator001.Value)
    )
    Scale RGB=&MakeVectorFromLinearColorRGB(Color=&ColorFromCurve(ColorCurve=DataInterface() {propertyValues {
        AlphaCurve {keys [<at 0 1 RCIM_Cubic>]}
        BlueCurve {keys [<at 0 1 RCIM_Cubic>,<at 0.2386 1 RCIM_Cubic>]}
        GreenCurve {keys [<at 0 0.3697 RCIM_Cubic>,<at 0.2386 0.2742 RCIM_Cubic>]}
        RedCurve {keys [<at 0 0.1628 RCIM_Cubic>,<at 0.2386 0.6238 RCIM_Cubic>]}
       }}))
    ScaleA=true
    ScaleRGB=true
   )
   &DynamicMaterialParameters(
    Index 0 Param 1=1
    Index 0 Param 2=1
    Index 0 Param 3=&FloatFromCurve(FloatCurve=DataInterface() {propertyValues {Curve {keys [{arriveTangent=1.3962, interpMode=RCIM_Cubic, leaveTangent=1.3962, tangentMode=RCTM_User, time=0.0032, value=0.136}, <at 1 1>]}}})
    Index 0 Param 4=&FloatFromCurve(
     FloatCurve=DataInterface() {propertyValues {Curve {keys [{arriveTangent=-1.0048, interpMode=RCIM_Cubic, leaveTangent=-1.0048, tangentMode=RCTM_User, time=-0.0046, value=1.0113}, <at 0.7 0.3>]}}}
     Scale Curve=0.5
    )
    Module.Param0WriteEnabled=false
    Module.Param1WriteEnabled=true
    Module.Param2WriteEnabled=true
    Module.Param3WriteEnabled=true
   )
   &ScaleMeshSize(Scale Factor=&VectorFromFloat(Value=&FloatFromCurve(
      FloatCurve=DataInterface() {propertyValues {Curve {keys [{arriveTangent=1.9256, interpMode=RCIM_Cubic, leaveTangent=1.9256, tangentMode=RCTM_User, time=0, value=0}, <at 1 1 RCIM_Cubic>]}}}
      Scale Curve=&Lerp_Float(A=0.9, B=1.4, Alpha=$System.FN_Integrator001.Value)
     )))
  ]
  rendererClasses [&NiagaraMeshRendererProperties]
  renderers [{rendererClass=&NiagaraMeshRendererProperties, rendererIndex=0}]
 }
 Corona {
  emitterUpdateScript [
   &EmitterState(
    Enable Distance Culling=true
    Life Cycle Mode=::NewEnumerator1
    Loop Behavior=::NewEnumerator1
    Loop Duration=2
    Max Distance Response=::SleepAndClearParticles
    MaxDistance=4000
    Module.bUseMaxDistance=true
    Module.bUseMinDistance=false
    Scalability Mode=::NewEnumerator1
    UseLoopDelay=false
   )
   &SpawnBurst_Instantaneous(Spawn Count=2, Use Spawn Probability=false, UseLoopCountLimit=false)
   &SpawnRate(SpawnRate=0.5, Use Spawn Probability=false) {enabled=false}
  ]
  particleSpawnScript [
   &FN_SetSpawnGroup(Number of Spawn Groups=2)
   &InitializeParticle(
    Color=<rgba 0.136 0.8505 2 0.5>
    Lifetime=&RandomRangeFloat(Minimum=3, Maximum=3.5)
    Mass=&RandomRangeFloat(Minimum=0.6, Maximum=1.2)
    Module.Write Color=true
    Module.Write Lifetime=true
    Module.Write Mass=true
    Module.Write Material Random=true
    Module.Write Position=false
    Module.Write Scale=false
    Module.Write SpriteRotation=true
    Module.Write SpriteSize=true
    Sprite Rotation=&NormalizedAngleToDegrees(NormalizedAngle=&RandomRangeFloat(Minimum=0, Maximum=1))
    Sprite Size=&Vector2DFromFloat(Value=100)
   )
   &SphereLocation(
    MaskSpawnGroup=false
    Override Seed=false
    Sphere Radius=1
    Surface Only=false
    UseNonUniformScale=false
    UseOffset=true
    UseSphereOrientation=false
   ) {enabled=false}
   &DynamicMaterialParameters(
    Index 0 Param 1=1
    Index 0 Param 2=&RandomRangeFloat(Minimum=0.05, Maximum=0.1)
    Index 0 Param 3=&RandomRangeFloat(Minimum=0.4, Maximum=0.6)
    Index 0 Param 4=1
    Module.Param0WriteEnabled=true
    Module.Param1WriteEnabled=true
    Module.Param2WriteEnabled=true
    Module.Param3WriteEnabled=true
   ) {enabled=false}
   /SetVariables(
    Particles.New Linear Color:LinearColor=<rgba 4.5334 2 40 0.08>
    Particles.SecondCoronaSize:Vector2f=&Vector2DFromFloat(Value=25)
    Particles.ThirdCoronaColor:LinearColor=<rgba 3.9 2 40 0.2>
    Particles.ThirdCoronaDyn:Vector4f=<1.1 0.2 1 0>
    Particles.ThirdCoronaSize:Vector2f=&Vector2DFromFloat(Value=50)
   ) {isSetParameters=true}
  ]
  particleUpdateScript [
   &ParticleState(Kill Particles When Lifetime Has Elapsed=false)
   &SkeletalMeshLocation(
    Bone / Socket Index=$Particles.SpawnGroup
    Bone Sampling Mode=::NewEnumerator4
    Inherit Velocity=false
    Interpolate Mesh Sampling=false
    MaskSpawnGroup=false
    Skeletal Mesh=$User.SkeletalMesh
   ) {module=SkeletalMeshLocation001}
   &SkeletalMeshLocation(
    Bone / Socket Index=$Particles.SpawnGroup
    Bone Sampling Mode=::NewEnumerator4
    Inherit Velocity=false
    Interpolate Mesh Sampling=false
    MaskSpawnGroup=false
    Sampled Position Offset=&Multiply_Vector(
     A=&Lerp_Vector(A=<0 11 12>, B=<0 -7 12>, Alpha=$System.Dive)
     B=&Lerp_Vector(A=<1 1 1>, B=<1 1 3>, Alpha=$Particles.NormalizedAge)
    )
    Skeletal Mesh=$User.SkeletalMesh
   ) {enabled=false, module=SkeletalMeshLocation002}
   &SkeletalMeshLocation(
    Bone / Socket Index=$Particles.SpawnGroup
    Bone Sampling Mode=::NewEnumerator4
    Inherit Velocity=false
    Interpolate Mesh Sampling=false
    MaskSpawnGroup=false
    Sampled Position Offset=&Multiply_Vector(
     A=&Lerp_Vector(A=<0 11 12>, B=<0 -7 12>, Alpha=$System.Dive)
     B=&Lerp_Vector(A=<1 1 1>, B=<1 1 3>, Alpha=$Particles.NormalizedAge)
    )
    Skeletal Mesh=$User.SkeletalMesh
   ) {enabled=false}
   &DynamicMaterialParameters(
    Index 0 Param 1=1
    Index 0 Param 2=&RandomRangeFloat(Minimum=0, Maximum=0.2)
    Index 0 Param 3=1
    Module.Param0WriteEnabled=true
    Module.Param1WriteEnabled=true
    Module.Param2WriteEnabled=true
    Module.Param3WriteEnabled=false
   ) {enabled=false, module=DynamicMaterialParameters001}
   &ScaleSpriteSize(
    Uniform Curve Scale=0.25
    Uniform Curve Sprite Scale=DataInterface() {propertyValues {Curve {keys [{tangentMode=RCTM_User, time=1, value=1}]}}}
   )
   &ScaleColor(Scale Alpha=0.5, ScaleA=true, ScaleRGB=true)
   &SpriteRotationRate(Rotation Rate=&RandomRangeFloat(Minimum=-5, Maximum=5))
  ]
  rendererClasses [&NiagaraSpriteRendererProperties]
  renderers [
   {rendererClass=&NiagaraSpriteRendererProperties}
   {rendererClass=&NiagaraSpriteRendererProperties, rendererIndex=1}
   {rendererClass=&NiagaraSpriteRendererProperties, rendererIndex=2}
  ]
 }
 Embers {
  simTarget=GPUComputeSim
  emitterUpdateScript [
   &EmitterState()
   &SpawnRate(Spawn Probability=0.5, SpawnRate=10, Use Spawn Probability=true)
  ]
  particleSpawnScript [
   &FN_SetSpawnGroup()
   /Niagara/Modules/Spawn/Initialization/V2/InitializeParticle(
    Color Maximum=<rgba 0.0269 0.25 0.974 1>
    Color Minimum=<rgba 1 0.1978 0.7007 1>
    Color Mode=::NewEnumerator2
    Lifetime Max=1.2
    Lifetime Min=0.8
    Lifetime Mode=::NewEnumerator1
    Sprite Size Mode=::NewEnumerator1
    Uniform Sprite Size Max=15
    UsePositionOffset=false
    Write Material Random=true
   ) {module=InitializeParticle001}
   &ShapeLocation(
    Hemisphere Distribution=<1 0.025>
    MaskSpawnGroup=true
    Offset Mode=::NewEnumerator0
    Override Seed=false
    Rotation Mode=::NewEnumerator2
    Sphere Radius=50
    Sphere Surface Distribution=1
    UseOwnerScale=true
    Yaw / Pitch / Roll=<0 0 0.025>
   )
   &AddVelocity(
    Module.Use Falloff Distance=false
    Origin Offset=<0 10 100>
    Override Seed=false
    Velocity Mode=::NewEnumerator1
    Velocity Speed=&FloatFromCurve(
     FloatCurve=DataInterface() {propertyValues {Curve {keys [<at 0.25 5 RCIM_Cubic>, {interpMode=RCIM_Cubic, leaveTangent=-228.4063, time=0.75, value=100}]}}}
     CurveIndex=&RandomRangeFloat(Minimum=0.2, Maximum=0.6)
     Scale Curve=&Lerp_Float(A=-3, B=-20, Alpha=$System.Timeline.Time)
    )
   )
   &AddVelocity(
    Velocity=$Particles.ShapeLocation.ShapeBitangent
    Velocity Speed Scale=&RandomRangeFloat(Minimum=5, Maximum=-5)
   ) {module=AddVelocity001}
   &SkeletalMeshLocation(
    Bone / Socket Index=$Particles.SpawnGroup
    Bone Sampling Mode=::NewEnumerator4
    Inherit Velocity=false
    Interpolate Mesh Sampling=false
    MaskSpawnGroup=false
    Sampled Position Offset=&Multiply_Vector(
     A=&Lerp_Vector(A=<0 11 12>, B=<0 -7 12>, Alpha=$System.Dive)
     B=&Lerp_Vector(A=<1 1 1>, B=<1 1 3>, Alpha=$Particles.NormalizedAge)
    )
    Skeletal Mesh=$User.SkeletalMesh
   )
  ]
  particleUpdateScript [
   &ParticleState()
   &InheritSourceMovement(
    Applied Force Scale=<-0.5 -0.5 0>
    ApplyForce=true
    ApplyForceFalloff=false
    ApplyPosition=true
    ApplyPositionFalloff=false
    Module.LimitForceSpeed=false
   ) {module=InheritSourceMovement001}
   &Drag(Drag=&RandomRangeFloat(Minimum=0.5, Maximum=1), Use Linear Drag=true, Use Rotational Drag=false)
   &AccelerationForce(
    Acceleration=&Lerp_Vector(
     A=<0 0 25>
     B=<0 0 500>
     Alpha=&FloatFromCurve(
      FloatCurve=DataInterface() {propertyValues {Curve {keys [
          {arriveTangent=15.6865, interpMode=RCIM_Cubic, leaveTangent=15.6865, tangentMode=RCTM_User, time=0, value=0}
          <at 0.1408 0.9847 RCIM_Cubic>
          {arriveTangent=-0.0405, interpMode=RCIM_Cubic, leaveTangent=-0.0405, tangentMode=RCTM_User, time=0.5, value=0}
         ]}}}
      CurveIndex=$System.Timeline.NormalizedTime
     )
    )
    Coordinate Space=::Local
   )
   /Niagara/Modules/Update/Forces/CurlNoiseForce(
    Noise Frequency=1
    Noise Quality / Cost=::NewEnumerator0
    Noise Strength=&FloatFromCurve(
     FloatCurve=DataInterface() {propertyValues {Curve {keys [{interpMode=RCIM_Cubic, tangentMode=RCTM_User, time=0, value=1}, {interpMode=RCIM_Cubic, tangentMode=RCTM_User, time=1, value=-1}]}}}
     Scale Curve=150
    )
    Pan Noise=false
   )
   &RotateAroundPoint(
    Radius:NiagaraFloat=&FloatFromCurve(
     FloatCurve=DataInterface() {propertyValues {Curve {keys [{interpMode=RCIM_Cubic, tangentMode=RCTM_User, time=0, value=0.25}, {interpMode=RCIM_Cubic, tangentMode=RCTM_User, time=1, value=1}]}}}
     Scale Curve=6
    )
    Rotation Rate:NiagaraFloat=&RandomRangeFloat(Minimum=0.5, Maximum=1)
    Set Particle Position=false
   )
   &SolveForcesAndVelocity(Clamp Velocity=false, Limit Acceleration=false)
   /SetVariables(Particles.Rotational_Position:NiagaraPosition=&Add_Vector(
     A=Linked() {linkedVariable {name=Particles.Position, type=NiagaraPosition}}
     B=Linked() {linkedVariable {name=Output.RotateAroundPoint.OutputPosition, type=NiagaraPosition}}
    )) {isSetParameters=true}
   &ScaleColor(
    Scale Alpha=&FloatFromCurve(
     FloatCurve=DataInterface() {propertyValues {Curve {keys [
         {arriveTangent=26.7273, interpMode=RCIM_Cubic, leaveTangent=26.7273, tangentMode=RCTM_User, time=0, value=0}
         {interpMode=RCIM_Cubic, tangentMode=RCTM_User, time=0.0967, value=0.9767}
         {interpMode=RCIM_Cubic, tangentMode=RCTM_User, time=0.66, value=1}
         <at 1 0 RCIM_Cubic>
        ]}}}
     Scale Curve=&RandomRangeFloat(Minimum=0.5, Maximum=2)
    )
    ScaleA=true
    ScaleRGB=true
   )
   &ScaleSpriteSize(Scale Factor=&Vector2DFromFloat(Value=&FloatFromCurve(FloatCurve=DataInterface() {propertyValues {Curve {keys [<at -0.0046 0>,<at 0.025 1>,<at 0.66 1>,<at 1.0011 0.002>]}}}, Scale Curve=0.5)))
  ]
  rendererClasses [&NiagaraSpriteRendererProperties]
  renderers [{rendererClass=&NiagaraSpriteRendererProperties}]
 }
 Fragments_Burst {
  emitterUpdateScript [
   &EmitterState(Life Cycle Mode=::NewEnumerator1, Loop Behavior=::NewEnumerator1, Loop Duration=0.5, UseLoopDelay=false)
   &SpawnBurst_Instantaneous(
    Spawn Count=&SelectIntFromArray(Direct Array Index=$User.SaplingAge, Int Selection Array=DataInterface() {propertyValues {IntData [2,3,5]}})
    Use Spawn Probability=false
   )
   &SpawnBurst_Instantaneous(
    Spawn Count=&SelectIntFromArray(Direct Array Index=$User.SaplingAge, Int Selection Array=DataInterface() {propertyValues {IntData [5,7,10]}})
    Spawn Time=0.02
    Use Spawn Probability=false
   ) {module=SpawnBurst_Instantaneous001}
   &SpawnBurst_Instantaneous(Spawn Count=20, Spawn Time=0.04, Use Spawn Probability=false) {module=SpawnBurst_Instantaneous002}
  ]
  particleSpawnScript [
   /Niagara/Modules/Spawn/Initialization/InitializeParticle(
    Color=&MakeLinearColorFromVectorAndFloat("Vector (RGB)"=<25 25 25>, "Float (Alpha)"=&UniformRangedFloat(Minimum=0.7, Maximum=1))
    Lifetime=&UniformRangedFloat(Minimum=0.25, Maximum=1)
    Mesh Scale=&VectorFromFloat(Value=&UniformRangedFloat(Minimum=0.1, Maximum=0.3))
    Module.Write Color=true
    Module.Write Lifetime=true
    Module.Write Mass=true
    Module.Write Material Random=true
    Module.Write Position=false
    Module.Write Scale=true
    Module.Write SpriteRotation=false
    Module.Write SpriteSize=false
    Position=&SimulationPosition()
    Sprite Size=<5 5>
   )
   &ShapeLocation(
    Cylinder Height=&SelectFloatFromArray(Direct Array Index=$User.SaplingAge, Float Selection Array=DataInterface() {propertyValues {FloatData [100,180,250]}})
    Cylinder Radius=&SelectFloatFromArray(Direct Array Index=$User.SaplingAge, Float Selection Array=DataInterface() {propertyValues {FloatData [25,30,40]}})
    MaskSpawnGroup=false
    Offset=&SelectVectorFromArray(Direct Array Index=$User.SaplingAge, Vector Selection Array=DataInterface() {propertyValues {FloatData [<0 0 50>, <0 0 90>, <0 0 125>]}})
    Offset Mode=::NewEnumerator0
    Override Seed=false
    Shape Primitive=::NewEnumerator1
    Surface Only=false
    UseOwnerScale=true
   )
   &SphereLocation(
    MaskSpawnGroup=false
    Override Seed=false
    Sphere Radius=&UniformRangedFloat(Minimum=0, Maximum=20)
    Surface Only=false
    UseNonUniformScale=false
    UseOffset=false
    UseSphereOrientation=false
   )
   &AddVelocity(Velocity=&Multiply_Vector(
     A=$Output.SphereLocation.SphereNormal
     B=&VectorFromFloat(Value=&FloatFromCurve(
       FloatCurve=DataInterface() {propertyValues {Curve {keys [<at 0 36 RCIM_Cubic>, {arriveTangent=-36, interpMode=RCIM_Cubic, leaveTangent=-36, time=0.37, value=12}, <at 1 0 RCIM_Cubic>]}}}
       Scale Curve=5
      ))
    ))
   &AddVelocityInCone(
    Cone Angle=&NormalizedAngleToDegrees(NormalizedAngle=&Lerp_Float(A=0, B=1, Alpha=$Particles.NormalizedAge))
    Cone Axis=<0 0 -1>
    Module.Use Velocity Falloff On Cone Axis=false
    Velocity Strength=100
   )
   /Niagara/Modules/Spawn/Orientation/V2/InitialMeshOrientation(
    Rotation=&UniformRangedVector(Minimum=<0 0 0>, Maximum=<1 1 1>)
    Use Orientation Vector=true
    Use Rotation Vector=true
   )
  ]
  particleUpdateScript [
   &ParticleState()
   &AccelerationForce(Acceleration=<0 0 700>)
   &VectorNoiseForce(
    Force Amount=&UniformRangedFloat(Minimum=625, Maximum=2000)
    Module.Use Random Noise Time Interval=true
    Random Noise Time Interval=0.15
   ) {enabled=false}
   &Drag(Use Linear Drag=true, Use Rotational Drag=false)
   &SolveForcesAndVelocity(Clamp Velocity=false, Limit Acceleration=false)
   /Niagara/Modules/Update/Orientation/V2/UpdateMeshOrientation(
    Rotation Rate:NiagaraFloat=&FloatFromCurve(FloatCurve=DataInterface() {propertyValues {Curve {keys [{arriveTangent=-2.8458, interpMode=RCIM_Cubic, leaveTangent=-2.8458, tangentMode=RCTM_User, time=0, value=1}, <at 1 0.2 RCIM_Cubic>]}}})
    Rotation Vector:Vector3f=<2 0.4 0.1>
   )
   &ScaleMeshSize(Scale Factor=&VectorFromFloat(Value=&FloatFromCurve(FloatCurve=DataInterface() {propertyValues {Curve {keys [<at 0 0 RCIM_Cubic>,<at 0.25 2 RCIM_Cubic>,<at 1 0 RCIM_Cubic>]}}}, Scale Curve=0.5)))
   &ScaleColor(
    Scale Alpha=&FloatFromCurve(FloatCurve=DataInterface() {propertyValues {Curve {keys [<at 0.7 1 RCIM_Cubic>,<at 1 0>]}}})
    Scale RGB=&MakeVectorFromLinearColorRGB(Color=&ColorFromCurve(ColorCurve=DataInterface() {propertyValues {
        AlphaCurve {keys []}
        BlueCurve {keys [<at 0 8 RCIM_Cubic>,<at 0.8 0.25>,<at 1 0 RCIM_Cubic>]}
        GreenCurve {keys [<at 0 1 RCIM_Cubic>,<at 0.8 0.25>,<at 1 0 RCIM_Cubic>]}
        RedCurve {keys [<at 0 1 RCIM_Cubic>,<at 0.8 0.25>,<at 1 0 RCIM_Cubic>]}
       }}))
    ScaleA=true
    ScaleRGB=true
   )
  ]
  rendererClasses [&NiagaraMeshRendererProperties]
  renderers [{rendererClass=&NiagaraMeshRendererProperties, rendererIndex=0}]
 }
 Glow_Device {
  emitterUpdateScript [
   &EmitterState(Life Cycle Mode=::NewEnumerator1, Loop Behavior=::NewEnumerator1, Loop Duration Mode=::NewEnumerator1, UseLoopDelay=false)
   &SpawnBurst_Instantaneous(
    Spawn Count=4
    Spawn Probability=HlslExpression() {hlslExpression="Emitter.LoopCount == 0"}
    Use Spawn Probability=true
    UseLoopCountLimit=false
   ) {module=SpawnBurst_Instantaneous001}
  ]
  particleSpawnScript [
   &FN_SetSpawnGroup()
   /Niagara/Modules/Spawn/Initialization/V2/InitializeParticle(
    Color=&HueShiftLinearColor(Hue=0.72, Saturation=3, Value=0.7, Linear Color=<rgba 2.75 15 0 0.5>)
    Sprite Rotation Mode=::NewEnumerator2
    Sprite Size Mode=::NewEnumerator3
    Uniform Sprite Size=400
    UsePositionOffset=false
    Write Material Random=true
   ) {module=InitializeParticle001}
   &CameraOffset(Camera Offset Amount=30)
   /SetVariables(Particles.ColorGLow:LinearColor=<rgba 0 0.25 5 0.1>, Particles.GlowSize=<100 100>) {isSetParameters=true}
  ]
  particleUpdateScript [
   &ParticleState(Kill Particles When Lifetime Has Elapsed=false, Let Infinitely Lived Particles Die When Emitter Deactivates=true)
   &SkeletalMeshLocation(
    MaskSpawnGroup=false
    Mesh Sampling Type=::NewEnumerator5
    OverrideRandomSeed=false
    Sampled Position Offset=<0 0 7>
    Skeletal Mesh=$User.SkeletalMesh
   ) {enabled=false}
   &SkeletalMeshLocation(
    Bone / Socket Index=$Particles.SpawnGroup
    Bone Sampling Mode=::NewEnumerator4
    Inherit Velocity=false
    Interpolate Mesh Sampling=false
    MaskSpawnGroup=false
    Skeletal Mesh=$User.SkeletalMesh
   ) {module=SkeletalMeshLocation001}
   &ScaleSpriteSize(
    Uniform Curve Index=$System.NormalizedLoopAge
    Uniform Curve Sprite Scale=DataInterface() {propertyValues {Curve {keys [
        {interpMode=RCIM_Cubic, tangentMode=RCTM_User, time=0, value=0.6}
        <at 0.5 0.7>
        {interpMode=RCIM_Cubic, tangentMode=RCTM_User, time=1, value=0.6}
       ]}}}
   )
   &SpriteRotationRate()
   &ScaleColor(
    Scale Alpha=&Multiply_Float(
     A=&Multiply_Float(A=0.4, B=0.02)
     B=&Multiply_Float(
      A=&Sine(Normalized Angle=$Emitter.Age, Period=0.7, Scale=0.2, Bias=1)
      B=&Sine(Normalized Angle=$Emitter.Age, Period=1, Scale=0.3, Bias=1)
     )
    )
    ScaleA=true
    ScaleRGB=true
   )
   &ScaleColor(
    Color Value To Scale=$Particles.Initial.ColorGLow
    Scale Alpha=&FloatFromCurve(
     FloatCurve=DataInterface() {propertyValues {Curve {keys [<at 0 0 RCIM_Cubic>,<at 0.15 1>,<at 0.6 1 RCIM_Cubic>,<at 1 0 RCIM_Cubic>]}}}
     Scale Curve=0.02
    )
    Scale RGB=<0.5 0.5 0.5>
    ScaleA=true
    ScaleRGB=true
   ) {enabled=false, module=ScaleColor001}
  ]
  rendererClasses [&NiagaraSpriteRendererProperties]
  renderers [
   {rendererClass=&NiagaraSpriteRendererProperties}
   {rendererClass=&NiagaraSpriteRendererProperties, rendererIndex=1}
  ]
 }
 LightGlow {
  emitterUpdateScript [
   &EmitterState()
   &SpawnBurst_Instantaneous(
    Spawn Count=&MakeIntFromFloat(Float=1)
    Spawn Probability=HlslExpression() {hlslExpression="Emitter.LoopCount == 0"}
    Use Spawn Probability=true
   )
  ]
  particleSpawnScript [
   &FN_SetSpawnGroup()
   /Niagara/Modules/Spawn/Initialization/InitializeParticle(
    Color=<rgba 0 0.6844 4 0.35>
    Mesh Scale=&VectorFromFloat(Value=40)
    Module.Write Color=true
    Module.Write Lifetime=true
    Module.Write Mass=true
    Module.Write Material Random=true
    Module.Write Position=true
    Module.Write Scale=true
    Module.Write SpriteRotation=false
    Module.Write SpriteSize=true
    Position=&SimulationPosition()
    Sprite Size=&Vector2DFromFloat(Value=40)
   )
   /SetVariables(Particles.MeshOrientation=&OrientMeshToVector(Look At Direction=<0 0 -1>, Reference Axis Vector=<1 0 0>)) {isSetParameters=true}
   /SetVariables(Particles.RotationalVelocity:Vector3f=&UniformRangedVector(Minimum=<-0.4 0 0>, Maximum=<0.4 0 0>)) {isSetParameters=true}
   /SetVariables(Particles.LightColor=<rgba 0 4 3.8 1>) {isSetParameters=true}
   /SetVariables(Particles.LightExponent=2) {isSetParameters=true}
  ]
  particleUpdateScript [
   &ParticleState(Kill Particles When Lifetime Has Elapsed=false, Let Infinitely Lived Particles Die When Emitter Deactivates=true)
   &SkeletalMeshLocation(
    Bone / Socket Index=4
    Bone Sampling Mode=::NewEnumerator4
    MaskSpawnGroup=false
    Sampled Position Offset=&Lerp_Vector(A=<-20 -50 0>, B=<-125 -65 0>, Alpha=$User.PlayerSpeed)
    Skeletal Mesh=$User.SkeletalMesh
   )
   &MeshRotationForce(Rotation=&UniformRangedVector(Minimum=<-1 0 0>, Maximum=<1 0 0>))
   &SolveForcesAndVelocity(Clamp Velocity=false, Limit Acceleration=false)
   &SolveRotationalForcesAndVelocity()
   /SetVariables(Particles.RotationLimited:Vector3f=&self:LimitVelocity(
     Max:Vector3f=&VectorFromFloat(Value=0.4)
     Min:Vector3f=&VectorFromFloat(Value=-0.4)
    )) {isSetParameters=true}
   /SetVariables(Particles.RotationalVelocity:Vector3f=$Particles.RotationLimited) {isSetParameters=true}
  ]
  rendererClasses [&NiagaraSpriteRendererProperties, &NiagaraLightRendererProperties]
  renderers [
   {rendererClass=&NiagaraSpriteRendererProperties}
   {rendererClass=&NiagaraLightRendererProperties, rendererIndex=1}
  ]
 }
 Mesh_Trails {
  emitterUpdateScript [
   &EmitterState(
    Inactive Response=::NewEnumerator1
    Life Cycle Mode=::NewEnumerator1
    Loop Behavior=::NewEnumerator1
    Loop Duration Mode=::NewEnumerator1
    UseLoopDelay=false
   )
   &SpawnBurst_Instantaneous(Spawn Count=4, Use Spawn Probability=false)
  ]
  particleSpawnScript [
   &FN_SetSpawnGroup(Number of Spawn Groups=5)
   /Niagara/Modules/Spawn/Initialization/V2/InitializeParticle(
    Color=&LerpLinearColors(
     StartColor=<rgba 1 4 5 1>
     EndColor=<rgba 0 5 3.4377 0.5>
     LerpFactor=&SelectFloatFromArray(Direct Array Index=$Particles.SpawnGroup, Float Selection Array=DataInterface() {propertyValues {FloatData [0,0,1,1]}})
    )
    Mesh Scale=&SelectVectorFromArray(
     Direct Array Index=$Particles.SpawnGroup
     Vector Selection Array=DataInterface() {propertyValues {FloatData [
        <0.2 0.2 0.45>
        <0.2 0.2 0.45>
        <0.2 0.2 0.4>
        <0.2 0.2 0.4>
        <0.3 0.3 0.4>
        <0.4 0.4 0.4>
        <0.4 0.4 0.4>
        <0.4 0.4 0.4>
        <0.4 0.4 0.5>
        <0.5 0.5 0.5>
       ]}}
    )
    Mesh Scale Mode=::NewEnumerator4
    UsePositionOffset=false
    Write Material Random=true
   )
   &DynamicMaterialParameters(
    Index 0 Param 2=&RandomRangeFloat(Minimum=0, Maximum=10)
    Index 0 Param 3=30
    Index 0 Param 4=1
    Module.Param0WriteEnabled=true
    Module.Param1WriteEnabled=true
    Module.Param2WriteEnabled=false
    Module.Param3WriteEnabled=true
   )
   /SetVariables(Particles.RotatorSpeedRandom:NiagaraFloat=&RandomRangeFloat(Minimum=0.3, Maximum=-0.3)) {isSetParameters=true}
   /SetVariables(Particles.WPO=25) {isSetParameters=true}
   &CameraOffset(Camera Offset Amount=10)
   /Niagara/Modules/Spawn/Orientation/V3/InitialMeshOrientation(
    Orientation Axis=<0 0 1>
    Orientation Coordinate Space=::Local
    Orientation Vector=<0 0 1>
    Rotation=&MakeVector(X=0, Y=0, Z=&RandomRangeFloat(Minimum=0, Maximum=360))
    Rotation Coordinate Space=::NewEnumerator0
    Use Orientation Vector=false
    Use Rotation Vector=true
   )
   /SetVariables(Particles.VisibilityTag=&SelectIntFromArray(Direct Array Index=$Particles.SpawnGroup, Int Selection Array=DataInterface() {propertyValues {IntData [0,0,1,1]}})) {isSetParameters=true}
   /SetVariables(
    Particles.SmokeColor:LinearColor=<rgba 0 0 0 0.07>
    Particles.SmokeSize:Vector3f=<0.2 0.2 2>
    Particles.WispyColor:LinearColor=&LerpLinearColors(
     StartColor=<rgba 1.9389 0.9936 5 0.2>
     EndColor=<rgba 0.9936 2.3959 5 0.2>
     LerpFactor=&SelectFloatFromArray(Direct Array Index=$Particles.SpawnGroup, Float Selection Array=DataInterface() {propertyValues {FloatData [0,0,1,1]}})
    )
   ) {isSetParameters=true}
  ]
  particleUpdateScript [
   &ParticleState(Kill Particles When Lifetime Has Elapsed=false, Let Infinitely Lived Particles Die When Emitter Deactivates=true)
   /SetVariables(Particles.MeshOrientation=&MultiplyQuaternion(
     Quaternion A=&FindQuatBetweenTwoVectors(From Vector=<0 0 -1>, To Vector=$System.SkydiveVelocity)
     Quaternion B=$Particles.Initial.MeshOrientation
    )) {enabled=false, isSetParameters=true}
   &InterpolateOverTime(Rate Of Change=15, Target Vector Value:NiagaraPosition=$Particles.Position, Variable Type=::NewEnumerator2)
   /SetVariables(Particles.InterpolateDynamic:Vector4f=&MakeFloat4FromVectorAndFloat("Vector (RGB)"=$Output.InterpolateOverTime.Moving Average, "Float (Alpha)"=1)) {isSetParameters=true}
   &SkeletalMeshLocation(
    Bone / Socket Index=$Particles.SpawnGroup
    Bone Sampling Mode=::NewEnumerator4
    Inherit Velocity=false
    Interpolate Mesh Sampling=false
    MaskSpawnGroup=false
    Skeletal Mesh=$User.SkeletalMesh
   )
   /Game/Effects/Niagara/Modules/Skydiving/FN_SkydivePositionOffset_v1/FN_SkydivePositionOffset(
    Bone Orientation:Quat4f=$Output.SkeletalMeshLocation.SampledRotation
    Bone Position:Vector3f=$Output.SkeletalMeshLocation.SampledPosition
    Offset:NiagaraFloat=/Game/Effects/Niagara/DynamicInputs/Skydiving/SkydiveHandsFeetFloatModifier(Hands:NiagaraFloat=20, Feet:NiagaraFloat=45)
   ) {enabled=false}
   /Game/Effects/Niagara/Modules/Skydiving/FN_Skydive_Velocity_v4/FN_Skydive_Velocity(
    Bone Orientation:Quat4f=$Output.SkeletalMeshLocation.SampledRotation
    Bone Position:Vector3f=$Output.SkeletalMeshLocation.SampledPosition
    Dive:NiagaraFloat=$System.Dive
    FE Speed=1
    SpawnGroup:NiagaraInt32=$Particles.SpawnGroup
    Write to Intrinsic Properties=false
   ) {enabled=false}
   /SetVariables(Particles.MeshOrientation=&FindQuatBetweenTwoVectors(From Vector=<0.8 0.8 1.2>, To Vector=$Output.FN_Skydive_Velocity.Velocity)) {enabled=false, isSetParameters=true}
   &ScaleMeshSize(Scale Factor=&SelectVectorFromArray(
     Direct Array Index=$Particles.SpawnGroup
     Vector Selection Array=DataInterface() {propertyValues {FloatData [<0.8 0.8 1.2>, <0.8 0.8 1.2>, <0.5 0.5 0.4>, <0.5 0.5 0.4>]}}
    ))
   &ScaleMeshSize(Scale Factor=&FN_FrontEndVector3Modifier(Gameplay=<1 1 1>, Front End=<1.25 1.25 2>)) {module=ScaleMeshSize001}
   &SolveForcesAndVelocity(Clamp Velocity=false, Limit Acceleration=false)
   /SetVariables(Particles.FootScale:Vector3f=&Multiply_Vector(
     A=&Multiply_Vector(
      A=$Particles.Initial.Scale
      B=&SelectVectorFromArray(
       Direct Array Index=$Particles.SpawnGroup
       Vector Selection Array=DataInterface() {propertyValues {FloatData [<0.8 0.8 1.2>, <0.8 0.8 1.2>, <0.2 0.2 0.3>, <0.2 0.2 0.3>]}}
      )
     )
     B=&FN_FrontEndVector3Modifier(Gameplay=<1 1 1>, Front End=<1.25 1.25 2>)
    )) {isSetParameters=true}
   &ScaleMeshSize(Scale Factor=&SelectVectorFromArray(
     Direct Array Index=$Particles.SpawnGroup
     Vector Selection Array=DataInterface() {propertyValues {FloatData [<0.8 0.8 1.2>, <0.8 0.8 1.2>, <0.5 0.5 0.4>, <0.5 0.5 0.4>]}}
    )) {enabled=false, module=ScaleMeshSize002}
   &DynamicMaterialParameters(
    Index 0 Param 2=&RandomRangeFloat(Minimum=0, Maximum=5)
    Index 0 Param 3=&Multiply_Float(A=$Particles.WPO, B=&Lerp_Float(A=0.5, B=0.4, Alpha=$System.Dive))
    Index 2 Param 1=&Lerp_Float(A=0.2, B=1, Alpha=&Ceil_Float(Float=$System.Dive))
    Index 2 Param 2=&Lerp_Float(A=0, B=0.3, Alpha=&Ceil_Float(Float=$System.Dive))
    Module.Param0WriteEnabled=false
    Module.Param0WriteEnabled2=true
    Module.Param1WriteEnabled=true
    Module.Param1WriteEnabled2=true
    Module.Param2WriteEnabled=true
    Module.Param2WriteEnabled2=false
    Module.Param3WriteEnabled=false
    Module.Param3WriteEnabled2=false
    Write Parameter Index 2=true
   ) {module=DynamicMaterialParameters001}
   &ScaleColor(Scale RGB=<0 1 1>, ScaleA=true, ScaleRGB=true)
  ]
  rendererClasses [&NiagaraMeshRendererProperties]
  renderers [
   {rendererClass=&NiagaraMeshRendererProperties, rendererIndex=0}
   {rendererClass=&NiagaraMeshRendererProperties, rendererIndex=1}
   {rendererClass=&NiagaraMeshRendererProperties, rendererIndex=2}
   {rendererClass=&NiagaraMeshRendererProperties, rendererIndex=3}
   {rendererClass=&NiagaraMeshRendererProperties, rendererIndex=4}
  ]
 }
 MiniZeroPoint {
  emitterSpawnScript [&FN_PreserveSkydiveUserVariables() {module=PreserveUserVariables}]
  emitterUpdateScript [
   &EmitterState()
   &SpawnBurst_Instantaneous(
    Spawn Count=&MakeIntFromFloat(Float=1)
    Spawn Probability=HlslExpression() {hlslExpression="Emitter.LoopCount == 0"}
    Use Spawn Probability=true
   )
  ]
  particleSpawnScript [
   &FN_SetSpawnGroup()
   /Niagara/Modules/Spawn/Initialization/InitializeParticle(
    Color=<rgba 0 7.8907 40 1>
    Mesh Scale=&VectorFromFloat(Value=0.2)
    Module.Write Color=true
    Module.Write Lifetime=true
    Module.Write Mass=false
    Module.Write Material Random=true
    Module.Write Position=false
    Module.Write Scale=true
    Module.Write SpriteRotation=false
    Module.Write SpriteSize=true
    Sprite Size=<19 19>
   )
   /SetVariables(Particles.MeshOrientation=&OrientMeshToVector(
     Look At Direction=&UniformRangedVector(Minimum=<-1 -1 -1>, Maximum=<1 1 1>)
     Reference Axis Vector=<1 0 0>
    )) {isSetParameters=true}
   /SetVariables(Particles.RotationalVelocity:Vector3f=&UniformRangedVector(
     Minimum=&VectorFromFloat(Value=-0.4)
     Maximum=&VectorFromFloat(Value=0.4)
    )) {isSetParameters=true}
   /SetVariables(Particles.SpritePosition:Vector3f=&Add_Vector(A=Linked() {linkedVariable {name=Particles.Position, type=NiagaraPosition}}, B=<0 0.5 -0.5>)) {isSetParameters=true}
   &DynamicMaterialParameters(
    Index 0 Param 2=1
    Index 1 Param 1=0.02
    Module.Param0WriteEnabled=true
    Module.Param0WriteEnabled1=true
    Module.Param1WriteEnabled=true
    Module.Param1WriteEnabled1=true
    Module.Param2WriteEnabled=true
    Module.Param2WriteEnabled1=true
    Module.Param3WriteEnabled=true
    Module.Param3WriteEnabled1=true
    Write Parameter Index 1=true
   )
  ]
  particleUpdateScript [
   &ParticleState(Kill Particles When Lifetime Has Elapsed=false, Let Infinitely Lived Particles Die When Emitter Deactivates=true)
   &SkeletalMeshLocation(
    Bone / Socket Index=4
    Bone Sampling Mode=::NewEnumerator4
    MaskSpawnGroup=false
    Orientation Sampling=::NewEnumerator1
    Sampled Position Offset=&Lerp_Vector(A=<-20 -50 0>, B=<-125 -65 0>, Alpha=$User.PlayerSpeed)
    Skeletal Mesh=$User.SkeletalMesh
   )
   &MeshRotationForce(Rotation=&UniformRangedVector(Minimum=<-3 -3 -3>, Maximum=<3 3 3>)) {enabled=false}
   /Niagara/Modules/Update/Orientation/V3/UpdateMeshOrientation(Rotation Rate=-0.1, Rotation Vector=<0.5 1 0.2>)
   &SolveForcesAndVelocity(Clamp Velocity=false, Limit Acceleration=false)
   &SolveRotationalForcesAndVelocity()
   /SetVariables(Particles.RotationLimited:Vector3f=&self:LimitVelocity_01(
     Max:Vector3f=&VectorFromFloat(Value=0.4)
     Min:Vector3f=&VectorFromFloat(Value=-0.4)
    )) {enabled=false, isSetParameters=true}
   /SetVariables(Particles.RotationalVelocity:Vector3f=$Particles.RotationLimited) {enabled=false, isSetParameters=true}
  ]
  rendererClasses [&NiagaraMeshRendererProperties]
  renderers [
   {rendererClass=&NiagaraMeshRendererProperties, rendererIndex=0}
   {rendererClass=&NiagaraMeshRendererProperties, rendererIndex=1}
   {rendererClass=&NiagaraMeshRendererProperties, rendererIndex=2}
   {rendererClass=&NiagaraMeshRendererProperties, rendererIndex=3}
  ]
 }
 Shards_Body {
  emitterUpdateScript [
   &EmitterState()
   &SpawnRate(SpawnRate=10, Use Spawn Probability=false)
  ]
  particleSpawnScript [
   &FN_SetSpawnGroup(Number of Spawn Groups=2)
   &InitializeParticle(
    Color=&MakeLinearColorFromVectorAndFloat("Vector (RGB)"=<25 25 25>, "Float (Alpha)"=&UniformRangedFloat(Minimum=0.7, Maximum=1))
    Lifetime=&UniformRangedFloat(Minimum=0.4, Maximum=1)
    Mesh Scale=&VectorFromFloat(Value=&UniformRangedFloat(Minimum=0.1, Maximum=0.3))
    Module.Write Color=true
    Module.Write Lifetime=true
    Module.Write Mass=true
    Module.Write Material Random=true
    Module.Write Position=false
    Module.Write Scale=true
    Module.Write SpriteRotation=false
    Module.Write SpriteSize=false
    Position=&SimulationPosition()
    Sprite Size=<5 5>
   )
   &ShapeLocation(
    Cylinder Height=&SelectFloatFromArray(Direct Array Index=$User.SaplingAge, Float Selection Array=DataInterface() {propertyValues {FloatData [100,180,250]}})
    Cylinder Radius=&SelectFloatFromArray(Direct Array Index=$User.SaplingAge, Float Selection Array=DataInterface() {propertyValues {FloatData [25,30,40]}})
    MaskSpawnGroup=false
    Offset=&SelectVectorFromArray(Direct Array Index=$User.SaplingAge, Vector Selection Array=DataInterface() {propertyValues {FloatData [<0 0 50>, <0 0 90>, <0 0 125>]}})
    Offset Mode=::NewEnumerator0
    Override Seed=false
    Shape Primitive=::NewEnumerator1
    Surface Only=false
    UseOwnerScale=true
   ) {enabled=false}
   &SphereLocation(
    MaskSpawnGroup=false
    Override Seed=false
    Sphere Radius=2
    Surface Only=false
    UseNonUniformScale=false
    UseOffset=true
    UseSphereOrientation=false
   ) {enabled=false}
   &SkeletalMeshLocation(
    Apply Sampled Orientation to Mesh Particles=true
    Bone / Socket Index=4
    Bone Sampling Mode=::NewEnumerator4
    Inherit Velocity=false
    Interpolate Mesh Sampling=false
    MaskSpawnGroup=false
    Sampled Position Offset=<0 0 -40>
    Sampled Position Offset Coordinate Space:ENiagaraMeshTransforms=::NewEnumerator2
    Skeletal Mesh=$User.SkeletalMesh
   )
   &ShapeLocation(MaskSpawnGroup=false, Offset=<0 0 50>, Offset Mode=::NewEnumerator0, Override Seed=false, Sphere Radius=1, UseOwnerScale=true) {module=ShapeLocation001}
   &AddVelocity(Velocity=&Multiply_Vector(
     A=&RandomRangeVector(Minimum=<50 0 0>, Maximum=<100 1 1>)
     B=&Lerp_Vector(A=<1 0 0>, B=<-1 0 0>, Alpha=&MakeFloatFromInt(Int=$Particles.SpawnGroup))
    )) {module=AddVelocity003}
   &AddVelocity(Velocity=&Multiply_Vector(
     A=$Output.SphereLocation.SphereNormal
     B=&VectorFromFloat(Value=&FloatFromCurve(
       FloatCurve=DataInterface() {propertyValues {Curve {keys [
           {arriveTangent=-718.0126, interpMode=RCIM_Cubic, leaveTangent=-63.9312, tangentMode=RCTM_User, time=0, value=12}
           {arriveTangent=-0.53, interpMode=RCIM_Cubic, leaveTangent=-0.53, tangentMode=RCTM_User, time=0.4, value=1}
           <at 1 0 RCIM_Cubic>
          ]}}}
       Scale Curve=4
      ))
    )) {enabled=false, module=AddVelocity001}
   &AddVelocity(Velocity=&Multiply_Vector(
     A=&VectorFromFloat(Value=2)
     B=&UniformRangedVector(Minimum=<0 0 30>, Maximum=<0 0 130>)
    )) {enabled=false, module=AddVelocity002}
   &AddVelocity(Velocity=&Multiply_Vector(
     A=$Output.SphereLocation.SphereNormal
     B=&VectorFromFloat(Value=&UniformRangedFloat(Minimum=1, Maximum=3))
    )) {enabled=false}
   /Niagara/Modules/Spawn/Orientation/V2/InitialMeshOrientation(
    Rotation=&UniformRangedVector(Minimum=<0 0 0>, Maximum=<1 1 1>)
    Use Orientation Vector=true
    Use Rotation Vector=true
   )
   &AddVelocityInCone(
    Cone Angle=&NormalizedAngleToDegrees(NormalizedAngle=&Lerp_Float(A=0, B=1, Alpha=$Particles.NormalizedAge))
    Cone Axis=<0 0 -1>
    Module.Use Velocity Falloff On Cone Axis=false
    Velocity Strength=200
   ) {enabled=false}
  ]
  particleUpdateScript [
   &ParticleState()
   &InheritSourceMovement(
    Applied Force Scale=<-0.5 -0.5 0>
    ApplyForce=true
    ApplyForceFalloff=false
    ApplyPosition=true
    ApplyPositionFalloff=false
    Module.LimitForceSpeed=false
   ) {module=InheritSourceMovement001}
   &AccelerationForce() {enabled=false}
   &VectorNoiseForce(
    Force Amount=&Multiply_Float(
     A=&UniformRangedFloat(Minimum=2000, Maximum=4000)
     B=&FloatFromCurve(FloatCurve=DataInterface() {propertyValues {Curve {keys [<at 0 3 RCIM_Cubic>, {arriveTangent=-3, interpMode=RCIM_Cubic, leaveTangent=-3, time=0.27, value=1.5}, <at 1 0 RCIM_Cubic>]}}})
    )
    Module.Use Random Noise Time Interval=true
    Random Noise Time Interval=0.15
   ) {enabled=false, module=VectorNoiseForce001}
   &GravityForce(Gravity=&Lerp_Vector(A=<0 0 0>, B=<1 1 1800>, Alpha=$System.Timeline.Time))
   &SolveForcesAndVelocity(Clamp Velocity=false, Limit Acceleration=false)
   /Niagara/Modules/Update/Orientation/V2/UpdateMeshOrientation(
    Rotation Rate:NiagaraFloat=&FloatFromCurve(FloatCurve=DataInterface() {propertyValues {Curve {keys [{arriveTangent=-2.8458, interpMode=RCIM_Cubic, leaveTangent=-2.8458, tangentMode=RCTM_User, time=0, value=1}, <at 1 0.2 RCIM_Cubic>]}}})
    Rotation Vector:Vector3f=<2 0.4 0.1>
   )
   &ScaleMeshSize(Scale Factor=&VectorFromFloat(Value=&FloatFromCurve(FloatCurve=DataInterface() {propertyValues {Curve {keys [<at 0 0 RCIM_Cubic>,<at 0.25 2 RCIM_Cubic>,<at 1 0 RCIM_Cubic>]}}}, Scale Curve=0.7)))
   &ScaleColor(
    Scale Alpha=&FloatFromCurve(FloatCurve=DataInterface() {propertyValues {Curve {keys [<at 0.7 1 RCIM_Cubic>,<at 1 0>]}}})
    Scale RGB=&MakeVectorFromLinearColorRGB(Color=&ColorFromCurve(ColorCurve=DataInterface() {propertyValues {
        AlphaCurve {keys []}
        BlueCurve {keys [<at 0 8 RCIM_Cubic>,<at 0.8 0.25>,<at 1 0 RCIM_Cubic>]}
        GreenCurve {keys [<at 0 1 RCIM_Cubic>,<at 0.8 0.25>,<at 1 0 RCIM_Cubic>]}
        RedCurve {keys [<at 0 1 RCIM_Cubic>,<at 0.8 0.25>,<at 1 0 RCIM_Cubic>]}
       }}))
    ScaleA=true
    ScaleRGB=true
   )
  ]
  rendererClasses [&NiagaraMeshRendererProperties]
  renderers [{rendererClass=&NiagaraMeshRendererProperties, rendererIndex=0}]
 }
 Shards_Hands {
  emitterUpdateScript [
   &EmitterState()
   &SpawnRate(SpawnRate=10, Use Spawn Probability=false)
  ]
  particleSpawnScript [
   &InitializeParticle(
    Color=&MakeLinearColorFromVectorAndFloat("Vector (RGB)"=<25 25 25>, "Float (Alpha)"=&UniformRangedFloat(Minimum=0.7, Maximum=1))
    Lifetime=&UniformRangedFloat(Minimum=0.5, Maximum=1.5)
    Mesh Scale=&VectorFromFloat(Value=&UniformRangedFloat(Minimum=0.1, Maximum=0.3))
    Module.Write Color=true
    Module.Write Lifetime=true
    Module.Write Mass=true
    Module.Write Material Random=true
    Module.Write Position=false
    Module.Write Scale=true
    Module.Write SpriteRotation=false
    Module.Write SpriteSize=false
    Position=&SimulationPosition()
    Sprite Size=<5 5>
   )
   &FN_SetSpawnGroup(Number of Spawn Groups=2)
   &ShapeLocation(
    Cylinder Height=&SelectFloatFromArray(Direct Array Index=$User.SaplingAge, Float Selection Array=DataInterface() {propertyValues {FloatData [100,180,250]}})
    Cylinder Radius=&SelectFloatFromArray(Direct Array Index=$User.SaplingAge, Float Selection Array=DataInterface() {propertyValues {FloatData [25,30,40]}})
    MaskSpawnGroup=false
    Offset=&SelectVectorFromArray(Direct Array Index=$User.SaplingAge, Vector Selection Array=DataInterface() {propertyValues {FloatData [<0 0 50>, <0 0 90>, <0 0 125>]}})
    Offset Mode=::NewEnumerator0
    Override Seed=false
    Shape Primitive=::NewEnumerator1
    Surface Only=false
    UseOwnerScale=true
   )
   &SphereLocation(
    MaskSpawnGroup=false
    Override Seed=false
    Sphere Radius=10
    Surface Only=false
    UseNonUniformScale=false
    UseOffset=false
    UseSphereOrientation=false
   )
   &SkeletalMeshLocation(
    Bone / Socket Index=$Particles.SpawnGroup
    Bone Sampling Mode=::NewEnumerator4
    Inherit Velocity=false
    Interpolate Mesh Sampling=false
    MaskSpawnGroup=false
    Sampled Position Offset=&Multiply_Vector(
     A=&Lerp_Vector(A=<0 11 12>, B=<0 -7 12>, Alpha=$System.Dive)
     B=&Lerp_Vector(A=<1 1 1>, B=<1 1 3>, Alpha=$Particles.NormalizedAge)
    )
    Skeletal Mesh=$User.SkeletalMesh
   )
   &AddVelocity(Velocity=&Multiply_Vector(
     A=$Output.SphereLocation.SphereNormal
     B=&VectorFromFloat(Value=&FloatFromCurve(FloatCurve=DataInterface() {propertyValues {Curve {keys [
           {arriveTangent=-718.0126, interpMode=RCIM_Cubic, leaveTangent=-63.9312, tangentMode=RCTM_User, time=0, value=12}
           {arriveTangent=-0.53, interpMode=RCIM_Cubic, leaveTangent=-0.53, tangentMode=RCTM_User, time=0.4, value=1}
           <at 1 0 RCIM_Cubic>
          ]}}}))
    )) {module=AddVelocity001}
   &AddVelocity(Velocity=&Multiply_VectorByFloat(
     Vector=&UniformRangedVector(Minimum=<0 0 100>, Maximum=<0 0 350>)
     Float=&RandomRangeFloat(Minimum=0.3, Maximum=0.8)
    )) {module=AddVelocity002}
   &AddVelocity(Velocity=&Multiply_Vector(
     A=$Output.SphereLocation.SphereNormal
     B=&VectorFromFloat(Value=&UniformRangedFloat(Minimum=1, Maximum=3))
    )) {enabled=false}
   &InitialMeshOrientation(
    Rotation=&UniformRangedVector(Minimum=<0 0 0>, Maximum=<1 1 1>)
    Use Orientation Vector=true
    Use Rotation Vector=true
   )
   &AddVelocityInCone(
    Cone Angle=&NormalizedAngleToDegrees(NormalizedAngle=&Lerp_Float(A=0, B=1, Alpha=$Particles.NormalizedAge))
    Cone Axis=<0 0 -1>
    Module.Use Velocity Falloff On Cone Axis=false
    Velocity Strength=200
   ) {enabled=false}
  ]
  particleUpdateScript [
   &ParticleState()
   &InheritSourceMovement(
    Applied Force Scale=<-0.5 -0.5 0>
    ApplyForce=true
    ApplyForceFalloff=false
    ApplyPosition=true
    ApplyPositionFalloff=false
    Module.LimitForceSpeed=false
   ) {module=InheritSourceMovement001}
   &GravityForce(Gravity=&Lerp_Vector(A=<0 0 0>, B=<1 1 1800>, Alpha=$System.Timeline.Time))
   &AccelerationForce(Acceleration=<0 0 200>) {enabled=false}
   &AccelerationForce(
    Acceleration=&Lerp_Vector(
     A=<0 0 25>
     B=<0 0 500>
     Alpha=&FloatFromCurve(
      FloatCurve=DataInterface() {propertyValues {Curve {keys [
          {arriveTangent=15.6865, interpMode=RCIM_Cubic, leaveTangent=15.6865, tangentMode=RCTM_User, time=0, value=0}
          <at 0.1408 0.9847 RCIM_Cubic>
          {arriveTangent=-0.0405, interpMode=RCIM_Cubic, leaveTangent=-0.0405, tangentMode=RCTM_User, time=0.5, value=0}
         ]}}}
      CurveIndex=$System.Timeline.NormalizedTime
     )
    )
    Coordinate Space=::Local
   ) {module=AccelerationForce001}
   &CurlNoiseForce(
    Noise Frequency=1
    Noise Quality / Cost=::NewEnumerator0
    Noise Strength=&FloatFromCurve(
     FloatCurve=DataInterface() {propertyValues {Curve {keys [{interpMode=RCIM_Cubic, tangentMode=RCTM_User, time=0, value=1}, {interpMode=RCIM_Cubic, tangentMode=RCTM_User, time=1, value=-1}]}}}
     Scale Curve=150
    )
    Pan Noise=false
   )
   &Drag(Drag=&RandomRangeFloat(Minimum=0.5, Maximum=1), Use Linear Drag=true, Use Rotational Drag=false)
   &VectorNoiseForce(
    Force Amount=&Multiply_Float(
     A=&UniformRangedFloat(Minimum=2000, Maximum=4000)
     B=&FloatFromCurve(FloatCurve=DataInterface() {propertyValues {Curve {keys [<at 0 3 RCIM_Cubic>, {arriveTangent=-3, interpMode=RCIM_Cubic, leaveTangent=-3, time=0.27, value=1.5}, <at 1 0 RCIM_Cubic>]}}})
    )
    Module.Use Random Noise Time Interval=true
    Random Noise Time Interval=0.15
   ) {enabled=false, module=VectorNoiseForce001}
   &RotateAroundPoint(
    Radius:NiagaraFloat=&FloatFromCurve(
     FloatCurve=DataInterface() {propertyValues {Curve {keys [{interpMode=RCIM_Cubic, tangentMode=RCTM_User, time=0, value=0.25}, {interpMode=RCIM_Cubic, tangentMode=RCTM_User, time=1, value=1}]}}}
     Scale Curve=10
    )
    Rotation Rate:NiagaraFloat=&RandomRangeFloat(Minimum=0.5, Maximum=1)
    Set Particle Position=false
   )
   &SolveForcesAndVelocity(Clamp Velocity=false, Limit Acceleration=false)
   &UpdateMeshOrientation(
    Rotation Rate:NiagaraFloat=&FloatFromCurve(FloatCurve=DataInterface() {propertyValues {Curve {keys [{arriveTangent=-2.8458, interpMode=RCIM_Cubic, leaveTangent=-2.8458, tangentMode=RCTM_User, time=0, value=1}, <at 1 0.2 RCIM_Cubic>]}}})
    Rotation Vector:Vector3f=<2 0.4 0.1>
   )
   /SetVariables(Particles.Rotational_Position:NiagaraPosition=&Add_Vector(A=Linked() {linkedVariable {name=Particles.Position, type=NiagaraPosition}}, B=$Output.RotateAroundPoint.OutputPosition)) {isSetParameters=true}
   &ScaleMeshSize(Scale Factor=&VectorFromFloat(Value=&FloatFromCurve(FloatCurve=DataInterface() {propertyValues {Curve {keys [<at 0 0 RCIM_Cubic>,<at 0.25 2 RCIM_Cubic>,<at 1 0 RCIM_Cubic>]}}}, Scale Curve=0.7)))
   &ScaleColor(
    Scale Alpha=&FloatFromCurve(FloatCurve=DataInterface() {propertyValues {Curve {keys [<at 0.7 1 RCIM_Cubic>,<at 1 0>]}}})
    Scale RGB=&MakeVectorFromLinearColorRGB(Color=&ColorFromCurve(ColorCurve=DataInterface() {propertyValues {
        AlphaCurve {keys []}
        BlueCurve {keys [<at 0 8 RCIM_Cubic>,<at 0.8 0.25>,<at 1 0 RCIM_Cubic>]}
        GreenCurve {keys [<at 0 1 RCIM_Cubic>,<at 0.8 0.25>,<at 1 0 RCIM_Cubic>]}
        RedCurve {keys [<at 0 1 RCIM_Cubic>,<at 0.8 0.25>,<at 1 0 RCIM_Cubic>]}
       }}))
    ScaleA=true
    ScaleRGB=true
   )
  ]
  rendererClasses [&NiagaraMeshRendererProperties]
  renderers [{rendererClass=&NiagaraMeshRendererProperties, rendererIndex=0}]
 }
 Tendrils001 {
  emitterUpdateScript [
   &EmitterState(Life Cycle Mode=::NewEnumerator1, Loop Duration=4, UseLoopDelay=false)
   &SpawnBurst_Instantaneous(Use Spawn Probability=false, UseLoopCountLimit=false)
   &SpawnRate(Spawn Probability=0.3, SpawnRate=5, Use Spawn Probability=true)
  ]
  particleSpawnScript [
   /Niagara/Modules/Spawn/Initialization/V2/InitializeParticle(
    Color=<rgba 0.387 0.09 9 1>
    Lifetime=&RandomRangeFloat(Minimum=0.8, Maximum=1.5)
    Mesh Scale=&MakeVector(
     X=&RandomRangeFloat(Minimum=0.8, Maximum=1.333)
     Y=1
     Z=&RandomRangeFloat(Minimum=0.666, Maximum=1.666)
    )
    Mesh Scale Mode=::NewEnumerator4
    UsePositionOffset=false
    Write Material Random=true
   )
   /Niagara/Modules/Spawn/Orientation/V3/InitialMeshOrientation(
    Mesh Orientation Mode=::NewEnumerator2
    Orientation Vector=<0 0 1>
    Rotation=&MakeVector(X=-0.25, Y=0, Z=0)
    Use Rotation Vector=true
   )
   &ShapeLocation(MaskSpawnGroup=false, Override Seed=false, Sphere Radius=10, UseOwnerScale=true) {enabled=false}
   &AddVelocity(Module.Use Falloff Distance=false, Override Seed=false, Velocity Mode=::NewEnumerator1, Velocity Speed=0) {enabled=false}
   &AddVelocity(Velocity=<0 0 0>) {module=AddVelocity001}
   &ApplyInitialForces()
  ]
  particleUpdateScript [
   &ParticleState()
   &SkeletalMeshLocation(
    Apply Sampled Orientation to Mesh Particles=true
    Bone / Socket Index=4
    Bone Sampling Mode=::NewEnumerator4
    Inherit Velocity=false
    Interpolate Mesh Sampling=false
    MaskSpawnGroup=false
    Sampled Position Offset=&Lerp_Vector(A=<0 0 -40>, B=<0 0 -65>, Alpha=$System.Timeline.Time)
    Skeletal Mesh=$User.SkeletalMesh
   ) {enabled=false}
   &SkeletalMeshLocation(
    Apply Sampled Orientation to Mesh Particles=true
    Bone / Socket Index=4
    Bone Sampling Mode=::NewEnumerator4
    Inherit Velocity=false
    Interpolate Mesh Sampling=false
    MaskSpawnGroup=false
    Sampled Position Offset=&Lerp_Vector(
     A=&RandomRangeVector(Minimum=<0 20 30>, Maximum=<0 -30 30>)
     B=&RandomRangeVector(Minimum=<0 20 100>, Maximum=<0 -30 100>)
     Alpha=$Particles.NormalizedAge
    )
    Sampled Position Offset Coordinate Space:ENiagaraMeshTransforms=::NewEnumerator2
    Skeletal Mesh=$User.SkeletalMesh
   ) {module=SkeletalMeshLocation001}
   &DynamicMaterialParameters(
    Index 0 Param 1=&RandomRangeFloat(Minimum=0, Maximum=0)
    Index 0 Param 2=&FloatFromCurve(
     FloatCurve=DataInterface() {propertyValues {Curve {keys [{interpMode=RCIM_Cubic, tangentMode=RCTM_User, time=0, value=1}, {interpMode=RCIM_Cubic, tangentMode=RCTM_User, time=1, value=0.5}]}}}
     Scale Curve=&Multiply_Float(A=0.3, B=&RandomRangeFloat(Minimum=0.2, Maximum=0.3))
    )
    Index 0 Param 3=&RandomRangeFloat(Minimum=0.3, Maximum=1)
    Index 0 Param 4=&FloatFromCurve(
     FloatCurve=DataInterface() {propertyValues {Curve {keys [
         {interpMode=RCIM_Cubic, tangentMode=RCTM_User, time=0, value=0}
         {interpMode=RCIM_Cubic, tangentMode=RCTM_User, time=0.5, value=1}
         {tangentMode=RCTM_User, time=1, value=0}
        ]}}}
     Scale Curve=&Multiply_Float(A=0.3, B=&RandomRangeFloat(Minimum=0.8, Maximum=0.8))
    )
    Module.Param0WriteEnabled=true
    Module.Param1WriteEnabled=true
    Module.Param2WriteEnabled=true
    Module.Param3WriteEnabled=true
   )
   &ScaleColor(
    Scale Alpha=&FloatFromCurve(
     FloatCurve=DataInterface() {propertyValues {Curve {keys [
         {arriveTangent=6.8255, interpMode=RCIM_Cubic, leaveTangent=6.8255, tangentMode=RCTM_User, time=0, value=0}
         {interpMode=RCIM_Cubic, tangentMode=RCTM_User, time=0.333, value=1}
         {tangentMode=RCTM_User, time=1, value=0}
        ]}}}
     Scale Curve=0.2
    )
    ScaleA=true
    ScaleRGB=true
   )
   &ScaleMeshSize(Scale Factor=&Multiply_Vector(
     A=&VectorFromFloat(Value=&RandomRangeFloat(Minimum=1, Maximum=1))
     B=&MakeVector(
      X=&RandomRangeFloat(Minimum=0.5, Maximum=0.6)
      Y=200
      Z=&FloatFromCurve(
       FloatCurve=DataInterface() {propertyValues {Curve {keys [{interpMode=RCIM_Cubic, tangentMode=RCTM_User, time=0, value=1}, {interpMode=RCIM_Cubic, tangentMode=RCTM_User, time=1, value=2}]}}}
       Scale Curve=&Multiply_Float(A=3, B=&RandomRangeFloat(Minimum=0.3, Maximum=0.3))
      )
     )
    ))
  ]
  rendererClasses [&NiagaraMeshRendererProperties]
  renderers [{rendererClass=&NiagaraMeshRendererProperties, rendererIndex=0}]
 }
 Tendrils002 {
  emitterUpdateScript [
   &EmitterState(Life Cycle Mode=::NewEnumerator1, Loop Behavior=::NewEnumerator1, Loop Duration Mode=::NewEnumerator1, UseLoopDelay=false)
   &SpawnRate(Use Spawn Probability=false) {enabled=false}
   &SpawnBurst_Instantaneous(
    Spawn Count=&MakeIntFromFloat(Float=3)
    Spawn Probability=HlslExpression() {hlslExpression="Emitter.LoopCount == 0"}
    Use Spawn Probability=true
   )
  ]
  particleSpawnScript [
   &InitializeParticle(
    Color=<rgba 0 0.4197 5 2>
    Lifetime=&RandomRangeFloat(Minimum=4, Maximum=2.5)
    Mesh Scale=&MakeVector(
     X=&RandomRangeFloat(Minimum=0.56, Maximum=0.9331)
     Y=0.7
     Z=&RandomRangeFloat(Minimum=0.4662, Maximum=1.1662)
    )
    Mesh Scale Mode=::NewEnumerator4
    UsePositionOffset=false
    Write Material Random=true
   )
   &InitialMeshOrientation(
    Mesh Orientation Mode=::NewEnumerator7
    Rotation=&MakeVector(X=0, Y=0, Z=&RandomRangeFloat(Minimum=0, Maximum=1))
    Use Rotation Vector=true
   ) {enabled=false}
   &ShapeLocation(MaskSpawnGroup=false, Override Seed=false, Ring Radius=44, Shape Primitive=::NewEnumerator4, UseOwnerScale=true)
   &InitialMeshOrientation(
    Mesh Axis To Orient=<0 0 1>
    Mesh Orientation Mode=::NewEnumerator2
    Orientation Coordinate Space=::Local
    Rotation=<0 0 0.25>
    Rotation Coordinate Space=::NewEnumerator2
    Use Rotation Vector=true
   ) {module=InitialMeshOrientation001}
   /SetVariables(Particles.VisibilityTag=&SelectIntFromArray(
     Direct Array Index=&ReturnExecIndex()
     Int Selection Array=DataInterface() {propertyValues {IntData [0,1,2,0]}}
    )) {isSetParameters=true}
   &FN_SetSpawnGroup(Number of Spawn Groups=3)
  ]
  particleUpdateScript [
   &ParticleState(Kill Particles When Lifetime Has Elapsed=false, Loop Particles Lifetime=true)
   &ScaleMeshSize(Scale Factor=&MakeVector(
     X=&SelectFloatFromArray(Direct Array Index=$Particles.VisibilityTag, Float Selection Array=DataInterface() {propertyValues {FloatData [0.2,0.2,0.2]}})
     Y=10
     Z=&Sine(Normalized Angle=$Emitter.Age, Period=1, Scale=0.1, Bias=0.5)
    )) {module=ScaleMeshSize001}
   &ScaleMeshSize(Scale Factor=&MakeVector(
     X=&FloatFromCurve(
      FloatCurve=DataInterface() {propertyValues {Curve {keys [{interpMode=RCIM_Cubic, tangentMode=RCTM_User, time=0, value=1}, {interpMode=RCIM_Cubic, tangentMode=RCTM_User, time=1, value=2}]}}}
      Scale Curve=&RandomRangeFloat(Minimum=0.8, Maximum=1.2)
     )
     Y=1
     Z=&FloatFromCurve(
      FloatCurve=DataInterface() {propertyValues {Curve {keys [
          {arriveTangent=3.3908, interpMode=RCIM_Cubic, leaveTangent=3.3908, tangentMode=RCTM_User, time=0, value=1}
          {arriveTangent=-0.0405, interpMode=RCIM_Cubic, leaveTangent=-0.0405, tangentMode=RCTM_User, time=1, value=2}
         ]}}}
      Scale Curve=0.2
     )
    )) {module=ScaleMeshSize002}
   &SkeletalMeshLocation(
    Apply Sampled Orientation to Mesh Particles=true
    Bone / Socket Index=4
    Bone Sampling Mode=::NewEnumerator4
    Inherit Velocity=false
    Interpolate Mesh Sampling=false
    MaskSpawnGroup=false
    Sampled Position Offset=&SelectVectorFromArray(
     Direct Array Index=$Particles.VisibilityTag
     Vector Selection Array=DataInterface() {propertyValues {FloatData [<0 -15 0>, <0 -8 -5>, <0 2 -6>]}}
    )
    Sampled Position Offset Coordinate Space:ENiagaraMeshTransforms=::NewEnumerator2
    Skeletal Mesh=$User.SkeletalMesh
   )
   &DynamicMaterialParameters(Write Parameter Index 0=false)
   &ScaleColor(
    Scale Alpha=&FloatFromCurve(FloatCurve=DataInterface() {propertyValues {Curve {keys [
         {arriveTangent=6.8255, interpMode=RCIM_Cubic, leaveTangent=6.8255, tangentMode=RCTM_User, time=0, value=0}
         {interpMode=RCIM_Cubic, tangentMode=RCTM_User, time=0.333, value=1}
         {interpMode=RCIM_Cubic, tangentMode=RCTM_User, time=1, value=0}
        ]}}})
    Scale RGB=&VectorFromFloat(Value=&Add_Float(A=1, B=&Multiply_Float(A=$User.Nurtuted, B=2)))
    ScaleA=true
    ScaleRGB=true
   )
   /Niagara/Modules/Update/Orientation/V3/UpdateMeshOrientation(Rotation Rate=0, Rotation Vector=<0 0 0>) {enabled=false}
   &ScaleMeshSize(Scale Factor=&MakeVector(
     X=1
     Y=1
     Z=&FloatFromCurve(
      FloatCurve=DataInterface() {propertyValues {Curve {keys [
          {arriveTangent=4.1785, interpMode=RCIM_Cubic, leaveTangent=4.1785, tangentMode=RCTM_User, time=0, value=0.2}
          {arriveTangent=0.9995, interpMode=RCIM_Cubic, leaveTangent=0.9995, tangentMode=RCTM_User, time=0.2223, value=0.6825}
          {arriveTangent=-0.0405, interpMode=RCIM_Cubic, leaveTangent=-0.0405, tangentMode=RCTM_User, time=1, value=1}
         ]}}}
      Scale Curve=&Multiply_Float(A=2, B=0.95)
     )
    )) {enabled=false}
  ]
  rendererClasses [&NiagaraMeshRendererProperties]
  renderers [
   {rendererClass=&NiagaraMeshRendererProperties, rendererIndex=0}
   {rendererClass=&NiagaraMeshRendererProperties, rendererIndex=1}
   {rendererClass=&NiagaraMeshRendererProperties, rendererIndex=2}
  ]
 }
 Wave_Center {
  emitterUpdateScript [
   &EmitterState()
   &SpawnRate(SpawnRate=&Lerp_Float(A=40, B=60, Alpha=$System.FN_Integrator001.Value), Use Spawn Probability=false)
  ]
  particleSpawnScript [
   &InitializeParticle(
    Color=<rgba 1 1.1 1 0.4>
    Lifetime=&RandomRangeFloat(Minimum=0.3, Maximum=0.4)
    Sprite Size Max=<10 50>
    Sprite Size Min=<10 30>
    Sprite Size Mode=::NewEnumerator2
    UsePositionOffset=false
    Write Material Random=true
   )
   /SetVariables(Particles.Position=&Add_Vector(A=Linked() {linkedVariable {name=Engine.Owner.Position, type=NiagaraPosition}}, B=<0 0 55>)) {isSetParameters=true}
   &TorusLocation(Align Torus With Axis=false, Handle Radius=0, Large Radius=40, MaskSpawnGroup=false, Module.Override Randomness=false)
   /SetVariables(Particles.RandomFrame=&RandomRangeInt(Minimum=0, Maximum=63)) {isSetParameters=true}
  ]
  particleUpdateScript [
   &ParticleState()
   &InheritSourceMovement(
    Applied Force Scale=<-0.5 -0.5 0>
    ApplyForce=true
    ApplyForceFalloff=false
    ApplyPosition=true
    ApplyPositionFalloff=false
    Module.LimitForceSpeed=false
   ) {module=InheritSourceMovement001}
   &AddVelocityFromPoint(
    Module.Use Falloff Distance=false
    Origin Offset=<0 0 22>
    Velocity Strength=&FloatFromCurve(
     FloatCurve=DataInterface() {propertyValues {Curve {keys [
         {arriveTangent=-1.0543, interpMode=RCIM_Cubic, time=0.25, value=1}
         {arriveTangent=-5, interpMode=RCIM_Cubic, time=0.6, value=0}
         <at 0.7 -1 RCIM_Cubic>
        ]}}}
     Scale Curve=13
    )
   ) {enabled=false}
   &PointForce(
    Force Falloff Distance=155
    Force Origin=&Add_Vector(A=/Niagara/DynamicInputs/Helpers/SimulationPosition(), B=<0 0 50>)
    Force Strength=1500
    Module.Use Falloff=true
    Module.UseLinearForce=false
   )
   &GravityForce(Gravity=&MakeVector(
     X=0
     Y=0
     Z=&FloatFromCurve(
      FloatCurve=DataInterface() {propertyValues {Curve {keys [<at 0.4 0 RCIM_Cubic>, {arriveTangent=0.5303, interpMode=RCIM_Cubic, leaveTangent=0.5303, tangentMode=RCTM_User, time=0.5, value=1}]}}}
      Scale Curve=2500
     )
    ))
   &SolveForcesAndVelocity(Clamp Velocity=false, Limit Acceleration=false)
   &ScaleColor(
    Scale Alpha=&FloatFromCurve(
     FloatCurve=DataInterface() {propertyValues {Curve {keys [
         {arriveTangent=-0.7151, interpMode=RCIM_Cubic, leaveTangent=-0.7151, tangentMode=RCTM_User, time=0.2, value=0}
         {arriveTangent=-2.5, interpMode=RCIM_Cubic, leaveTangent=-2.5, time=0.6, value=1}
         {arriveTangent=-1.6379, interpMode=RCIM_Cubic, leaveTangent=-1.6379, tangentMode=RCTM_User, time=1, value=0}
        ]}}}
     Scale Curve=0.2
    )
    Scale RGB=&MakeVectorFromLinearColorRGB(Color=&ColorFromCurve(ColorCurve=DataInterface() {propertyValues {
        AlphaCurve {keys [<at 0.3745 0>, {arriveTangent=-5.0984, interpMode=RCIM_Cubic, leaveTangent=-5.0984, time=0.529, value=1}, <at 0.7251 0>]}
        BlueCurve {keys [<at 0.0463 5>]}
        GreenCurve {keys [<at 0.0463 0.2018>]}
        RedCurve {keys [<at 0.0463 1.1614>]}
       }}))
    ScaleA=true
    ScaleRGB=true
   )
   &ScaleSpriteSize(Scale Factor=&Vector2DFromFloat(Value=&FloatFromCurve(
      FloatCurve=DataInterface() {propertyValues {Curve {keys [{arriveTangent=-0.2263, interpMode=RCIM_Cubic, leaveTangent=-0.2263, tangentMode=RCTM_User, time=0, value=0}, <at 1 1 RCIM_Cubic>]}}}
      Scale Curve=3
     )))
   &FN_SubUVAnimation(End Frame=&Add_Integer(A=$Particles.RandomFrame, B=24), Start Frame=$Particles.RandomFrame)
  ]
  rendererClasses [&NiagaraSpriteRendererProperties]
  renderers [{rendererClass=&NiagaraSpriteRendererProperties}]
 }
 Wispy_Swirl {
  emitterUpdateScript [
   &EmitterState(
    Inactive Response=::NewEnumerator1
    Life Cycle Mode=::NewEnumerator1
    Loop Behavior=::NewEnumerator1
    Loop Duration Mode=::NewEnumerator1
    UseLoopDelay=false
   )
   &SpawnBurst_Instantaneous(Spawn Count=1, Use Spawn Probability=false)
  ]
  particleSpawnScript [
   /Niagara/Modules/Spawn/Initialization/InitializeParticle(
    Color=<rgba 0.699 0.14 1 0.025>
    Lifetime=0
    Mesh Scale=&VectorFromFloat(Value=0.08)
    Module.Write Color=true
    Module.Write Lifetime=true
    Module.Write Mass=true
    Module.Write Material Random=true
    Module.Write Position=true
    Module.Write Scale=true
    Module.Write SpriteRotation=false
    Module.Write SpriteSize=false
    Position=&Add_Vector(A=/Niagara/Functions/Localspace/SimulationPosition(), B=<0 0 -30>)
    Sprite Size=<5 5>
   )
   /Niagara/Modules/Spawn/Orientation/V2/InitialMeshOrientation(Orientation Axis:Vector3f=<0 0 0>, Orientation Vector=<0 0 -1>, Use Orientation Vector=true, Use Rotation Vector=true)
   &DynamicMaterialParameters(
    Index 0 Param 1=1
    Index 0 Param 2=0.1
    Index 0 Param 3=1
    Index 0 Param 4=1
    Module.Param0WriteEnabled=true
    Module.Param1WriteEnabled=true
    Module.Param2WriteEnabled=false
    Module.Param3WriteEnabled=false
   ) {module=DynamicMaterialParameters001}
   &SkeletalMeshLocation(
    Bone / Socket Index=4
    Bone Sampling Mode=::NewEnumerator4
    MaskSpawnGroup=false
    Orientation Sampling=::NewEnumerator1
    Sampled Position Offset=<20 -20 0>
    Skeletal Mesh=$User.SkeletalMesh
   ) {enabled=false}
   /SetVariables(Particles.DynamicMaterialParameter1:Vector4f=<10 -2 1 1>) {isSetParameters=true}
   /SetVariables(Particles.SmallRing_Size:Vector3f=&VectorFromFloat(Value=0.5)) {isSetParameters=true}
  ]
  particleUpdateScript [
   &ParticleState(Kill Particles When Lifetime Has Elapsed=false, Let Infinitely Lived Particles Die When Emitter Deactivates=true)
   &SkeletalMeshLocation(
    Bone / Socket Index=4
    Bone Sampling Mode=::NewEnumerator4
    MaskSpawnGroup=false
    Orientation Sampling=::NewEnumerator1
    Sampled Position Offset=&Lerp_Vector(A=<-20 -50 0>, B=<-125 -65 0>, Alpha=$User.PlayerSpeed)
    Skeletal Mesh=$User.SkeletalMesh
   ) {module=SkeletalMeshLocation001}
   /SetVariables(Particles.SmallRing_Position:Vector3f=&Multiply_Vector(A=<1 1 1>, B=&Add_Vector(A=Linked() {linkedVariable {name=Particles.Position, type=NiagaraPosition}}, B=<0 0 0>))) {isSetParameters=true}
   /SetVariables(Particles.SmallRing_Position2:Vector3f=&Multiply_Vector(
     A=&VectorFromFloat(Value=1)
     B=&Add_Vector(A=Linked() {linkedVariable {name=Particles.Position, type=NiagaraPosition}}, B=<0 0 -3>)
    )) {isSetParameters=true}
   &ScaleMeshSize(Scale Factor=&VectorFromFloat(Value=3))
   &ScaleColor(Scale Alpha=4, Scale RGB=<10 10 10>, ScaleA=true, ScaleRGB=true)
   &UpdateMeshOrientation(Rotation Rate=0.2)
  ]
  rendererClasses [&NiagaraMeshRendererProperties]
  renderers [
   {rendererClass=&NiagaraMeshRendererProperties, rendererIndex=0}
   {rendererClass=&NiagaraMeshRendererProperties, rendererIndex=1}
   {rendererClass=&NiagaraMeshRendererProperties, rendererIndex=2}
  ]
 }
}
summary {
 systemName=NS_Contrail_Dive_MeteorWomen
 emitters [
  {emitterName=Wave_Center, rendererClasses [&NiagaraSpriteRendererProperties]}
  {emitterName=Fragments_Burst, rendererClasses [&NiagaraMeshRendererProperties]}
  {emitterName=Shards_Body, rendererClasses [&NiagaraMeshRendererProperties]}
  {emitterName=Embers, simTarget=GPUComputeSim, rendererClasses [&NiagaraSpriteRendererProperties]}
  {emitterName=Corona, rendererClasses [&NiagaraSpriteRendererProperties]}
  {emitterName=Mesh_Trails, rendererClasses [&NiagaraMeshRendererProperties]}
  {emitterName=Glow_Device, rendererClasses [&NiagaraSpriteRendererProperties]}
  {emitterName=Shards_Hands, rendererClasses [&NiagaraMeshRendererProperties]}
  {emitterName=Bowl_Multi, rendererClasses [&NiagaraMeshRendererProperties]}
  {emitterName=LightGlow, rendererClasses [&NiagaraSpriteRendererProperties, &NiagaraLightRendererProperties]}
  {emitterName=MiniZeroPoint, rendererClasses [&NiagaraMeshRendererProperties]}
  {emitterName=Wispy_Swirl, rendererClasses [&NiagaraMeshRendererProperties]}
  {emitterName=Tendrils001, rendererClasses [&NiagaraMeshRendererProperties]}
  {emitterName=Tendrils002, rendererClasses [&NiagaraMeshRendererProperties]}
 ]
 userVariables [
  User.bIsFrontEnd {type=NiagaraBool}
  User.bIsFrontEndPreview {type=NiagaraBool}
  User.BurstDelay {type=NiagaraFloat, defaultValue=0, description=""}
  User.ColorChange {type=LinearColor, description="", defaultValue {struct=LinearColor, value=<rgba 1 1 1 1>}}
  User.FadeInDuration {type=NiagaraFloat, defaultValue=1, description=""}
  User.Hit {type=NiagaraBool, defaultValue=true}
  User.Nurtuted {type=NiagaraFloat, defaultValue=0, description=""}
  User.OffsetOrb {type=Vector3f, description="", defaultValue {struct=Vector3f, value=<0 0 -15>}}
  User.OffsetOrbOhhand {type=Vector3f, description="", defaultValue {struct=Vector3f, value=<0 0 15>}}
  User.Opacity {type=NiagaraFloat, defaultValue=0, description=""}
  User.ParaGlideLeanAlpha {type=NiagaraFloat, defaultValue=0, description=""}
  User.PlayerSpeed {type=NiagaraFloat, defaultValue=0, description=""}
  User.RibbonIDOffset {type=NiagaraInt32}
  User.SaplingAge {type=NiagaraInt32}
  User.SkeletalMesh {
   type=NiagaraDataInterfaceSkeletalMesh
   defaultValue {
    struct=NiagaraExt_VariableValue_DataInterface
    value {dataInterface=&self:NiagaraDataInterfaceSkeletalMesh_0}
   }
  }
  User.SkeletalMesh_Flow {
   type=NiagaraDataInterfaceSkeletalMesh
   defaultValue {
    struct=NiagaraExt_VariableValue_DataInterface
    value {dataInterface=&self:NiagaraDataInterfaceSkeletalMesh_3}
   }
  }
 ]
}
